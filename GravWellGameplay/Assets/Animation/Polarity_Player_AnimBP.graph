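# Polarity player Animation Blueprint for the Fortnite player skeleton, overriding item animation layers for leap, firing and shield poses.
# The full-body layer plays an angle-driven leap blendspace, then applies the Polarity aim offset while firing; the final-pose layer does likewise while the shield is held.
# Leap entry flags AirborneFromLeap, sending the jump layer into a fall loop that tracks FallTime and clears the flag on exit.
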AnimBlueprint /GravWellGameplay/Assets/Animation/Polarity_Player_AnimBP
target_skeleton=&Fortnite_M_Avg_Player_Skeleton
graphs [
 AnimGraph {nodes [Output Pose()]}
 EventGraph {nodes [
   TryGetPawnOwner()
   Event BlueprintUpdateAnimation()
  ]}
 ItemAimOffset {nodes [
   Input Pose!() -> (Pose, AimOffsetAlpha:double, Yaw:double, Pitch:double) {name=InPose_AimOffset}
   Input Pose!() -> (Pose) {name=InPose_PreAimOffsets}
   Output Pose() {layerGroup=Item Override}
  ]}
 ItemFallAdditive {nodes [
   Input Pose() {name=InPoseItemFallAdditive}
   Output Pose() {layerGroup=Item Override}
  ]}
 ItemFallLandAdditive {nodes [
   Input Pose() {name=InPoseItemFallLandAdditive}
   Output Pose() {layerGroup=Item Override}
  ]}
 ItemFallLevel2Override {nodes [
   Input Pose() {name=InPose_FallLevel2}
   Output Pose() {layerGroup=Item Override}
  ]}
 ItemFinalPoseOverride {nodes [
   n1 = Use cached pose 'ItemFinalPoseInput'()
   n2 = Get!() -> (AdjustedAim_ResultingYaw:float=~, AdjustedAim_ResultingPitch:float=~, AdjustedAim_ZeroOutPitchWeightCurveName:Name)
   Output Pose(Result=Blend Poses by bool(
     BlendPose_0=AimOffset Player 'Polarity_AO'!(X:float=@n2, Y:float=@n2.AdjustedAim_ResultingPitch, BasePose:Pose Link Structure=@n1, Alpha:float=1) -> (Pose:Pose Link Structure) {blendSpace=/GravWellGameplay/Assets/Animation/AO/Polarity_AO, alphaBoolBlend {blendOption=Linear}}
     BlendPose_1=Use cached pose 'ItemFinalPoseInput'()
     bActiveValue=Get HoldShieldActive() -> (HoldShieldActive:bool)
    ) {blendType=Linear}) {layerGroup=Item Override}
   Save cached pose 'ItemFinalPoseInput'(Pose=Input Pose() {name=InFinalPose}) {cacheName=ItemFinalPoseInput}
  ]}
 ItemFullBodyAimOffsets {nodes [
   Input Pose!() -> (Pose, FullBody_AimOffsetAlpha:double, Aim_Yaw:double, Aim_Pitch:double) {name=InPoseFullBodyAimOffsets}
   Input Pose!() -> (Pose) {name=InPosePreFullBodyAimOffsets}
   Output Pose() {layerGroup=Item Override}
  ]}
 ItemFullBodyOverride {nodes [
   n1 = Use cached pose 'Incoming with Leap'()
   n2 = Get() -> (AdjustedAim_ResultingYaw=~, AdjustedAim_ResultingPitch=~)
   Output Pose(Result=Blend Poses by bool(
     BlendPose_0=AimOffset Player 'Polarity_AO'(X=@n2, Y=@n2.AdjustedAim_ResultingPitch, BasePose=@n1, Alpha=1) {blendSpace=&Polarity_AO, alphaBoolBlend {blendOption=Linear}}
     BlendPose_1=Use cached pose 'Incoming with Leap'()
     BlendTime_0=0
     BlendTime_1=0.2
     bActiveValue=Get bIsFiring() -> (bIsFiring:bool)
    ) {blendType=Linear}) {layerGroup=Item Override}
   Save cached pose 'Incoming Pose'(Pose=Input Pose() {name=InPoseFullBody}) {cacheName=Incoming Pose}
   AnimGraphNode_StateMachine_2 = Leap() {is=title AnimGraphNode_StateMachine}
   Save cached pose 'Incoming with Leap'(Pose=@AnimGraphNode_StateMachine_2) {cacheName=Incoming with Leap}
  ]}
 Leap {
  entry=Empty
  path=ItemFullBodyOverride.AnimGraphNode_StateMachine_2.Leap
  ids=id
  nodes [
   AnimStateNode_0 = Empty() {is=title AnimStateNode}
   AnimStateNode_1 = Leap() {is=title AnimStateNode, bAlwaysResetOnEntry=true}
   AnimStateTransitionNode_0 = AnimStateTransitionNode()
   AnimStateTransitionNode_1 = AnimStateTransitionNode() {crossfadeDuration=0.33}
  ]
  links [
   {
    dst=Leap
    rule=ItemFullBodyOverride.AnimGraphNode_StateMachine_2.Leap.AnimStateTransitionNode_0.Transition
    src=Empty
    via=AnimStateTransitionNode_0
   }
   {
    dst=Empty
    rule=ItemFullBodyOverride.AnimGraphNode_StateMachine_2.Leap.AnimStateTransitionNode_1.Transition
    src=Leap
    via=AnimStateTransitionNode_1
   }
  ]
 }
 Empty {
  path=ItemFullBodyOverride.AnimGraphNode_StateMachine_2.Leap.AnimStateNode_0.Empty
  nodes [Output Animation Pose(Result=Use cached pose 'Incoming Pose'())]
 }
 Leap {
  path=ItemFullBodyOverride.AnimGraphNode_StateMachine_2.Leap.AnimStateNode_1.Leap
  nodes [
   n1 = Blendspace Player 'Polarity_LeapAbility_BS'!(X=Get LeapAngle() -> (LeapAngle:float)) -> (Pose) {
    bLoop=false
    bResetPlayTimeWhenBlendSpaceChanges=false
    blendSpace=&/Polarity_LeapAbility_BS
    groupName=Leap
    method=SyncGroup
   }
   Output Animation Pose(Result=@n1) {
    becomeRelevantFunction {bSelfContext=true, memberGuid=#1, memberName=OnLeapBecomeRelevant}
    node {becomeRelevantFunction {functionName=OnBecomeRelevant}}
   }
  ]
 }
 Transition {
  path=ItemFullBodyOverride.AnimGraphNode_StateMachine_2.Leap.AnimStateTransitionNode_0.Transition
  nodes [Result(bCanEnterTransition=Get bCanEnterLeap() -> (bCanEnterLeap:bool))]
 }
 Transition {
  path=ItemFullBodyOverride.AnimGraphNode_StateMachine_2.Leap.AnimStateTransitionNode_1.Transition
  nodes [Result(bCanEnterTransition=NOT Boolean(A=Get bLaunchAbilityActive() -> (bLaunchAbilityActive:bool)))]
 }
 ItemIdleOverride {nodes [
   Input Pose() {name=InPose_Idle}
   Output Pose() {layerGroup=Item Override}
  ]}
 ItemJetPackJumpAdditive {nodes [Output Pose() {layerGroup=Item Override}]}
 ItemJetPackStartAdditive {nodes [Output Pose() {layerGroup=Item Override}]}
 ItemJumpLoopAdditive {nodes [
   Input Pose() {name=InPoseItemJumpLoopAdditive}
   Output Pose() {layerGroup=Item Override}
  ]}
 ItemJumpOverride {nodes [
   AnimGraphNode_StateMachine_0 = Fall() {is=title AnimGraphNode_StateMachine}
   Output Pose(Result=@AnimGraphNode_StateMachine_0) {layerGroup=Item Override}
   Save cached pose 'JumpStartInput'(Pose=Input Pose() {name=InPoseJump}) {cacheName=JumpStartInput}
  ]}
 Fall {
  entry=Input Pose
  path=ItemJumpOverride.AnimGraphNode_StateMachine_0.Fall
  ids=id
  nodes [
   AnimStateNode_0 = Input Pose() {is=title AnimStateNode}
   AnimStateNode_1 = Leap Fall() {
    is=title AnimStateNode
    stateEntered {guid=#2}
    stateFullyBlended {guid=#3}
    stateLeft {guid=#4}
   }
   AnimStateTransitionNode_1 = AnimStateTransitionNode()
   AnimStateTransitionNode_2 = AnimStateTransitionNode()
  ]
  links [
   {
    dst=Leap Fall
    rule=ItemJumpOverride.AnimGraphNode_StateMachine_0.Fall.AnimStateTransitionNode_1.Transition
    src=Input Pose
    via=AnimStateTransitionNode_1
   }
   {
    dst=Input Pose
    rule=ItemJumpOverride.AnimGraphNode_StateMachine_0.Fall.AnimStateTransitionNode_2.Transition
    src=Leap Fall
    via=AnimStateTransitionNode_2
   }
  ]
 }
 Input Pose {
  path=ItemJumpOverride.AnimGraphNode_StateMachine_0.Fall.AnimStateNode_0.Input Pose
  nodes [Output Animation Pose(Result=Use cached pose 'JumpStartInput'())]
 }
 Leap Fall {
  path=ItemJumpOverride.AnimGraphNode_StateMachine_0.Fall.AnimStateNode_1.Leap Fall
  nodes [
   n1 = Sequence Player 'NoWep_InAir_Fall_Loop_Lvl2'() {
    becomeRelevantFunction {bSelfContext=true, memberGuid=#5, memberName=OnBecomeRelevant_Fall}
    node {
     sequence=/Game/Animation/Game/MainPlayer/Locomotion/Medium/Male/InAir/NoWep_InAir_Fall_Loop_Lvl2
     becomeRelevantFunction {functionName=OnBecomeRelevant}
     updateFunction {functionName=OnUpdate}
    }
    updateFunction {bSelfContext=true, memberGuid=#6, memberName=OnUpdate_Fall}
   }
   Output Animation Pose(Result=@n1) {
    node {stateExitFunction {functionName=OnStateExit}}
    stateExitFunction {bSelfContext=true, memberGuid=#7, memberName=OnStateExit_Fall}
   }
  ]
 }
 Transition {
  path=ItemJumpOverride.AnimGraphNode_StateMachine_0.Fall.AnimStateTransitionNode_1.Transition
  nodes [Result(bCanEnterTransition=Get AirborneFromLeap() -> (AirborneFromLeap:bool))]
 }
 Transition {
  path=ItemJumpOverride.AnimGraphNode_StateMachine_0.Fall.AnimStateTransitionNode_2.Transition
  nodes [Result(bCanEnterTransition=Get bCanExitLeapFall() -> (bCanExitLeapFall:bool))]
 }
 ItemJumpUpAdditive {nodes [
   Input Pose() {is=title AnimGraphNode_LinkedInputPose, name=InPoseItemJumpUpAdditive}
   Output Pose() {layerGroup=Item Override}
  ]}
 ItemPreAimOffsets {nodes [
   Input Pose() {name=InPosePreAimOffsets}
   Output Pose() {layerGroup=Item Override}
  ]}
 ItemPreIK {nodes [
   Input Pose() {name=InPose_PreIK}
   Output Pose() {layerGroup=Item Override}
  ]}
 ItemRightHandIK {nodes [
   Input Pose() {name=InPosePreRightHandIK}
   Input Pose() {name=InPosePostRightHandIK}
   Output Pose() {layerGroup=Item Override}
  ]}
 ItemSkeletalControl {nodes [
   Input Pose()
   Output Pose() {layerGroup=Item Override}
  ]}
 ItemSwimJumpFallLoop {nodes [Output Pose() {layerGroup=Item Override}]}
 ItemSwimJumpSurfaceEnd {nodes [Output Pose() {layerGroup=Item Override}]}
 ItemSwimJumpSurfaceLoop {nodes [Output Pose() {layerGroup=Item Override}]}
 ItemUpperBody {nodes [
   Input Pose!() -> (Pose, InputParam:Fort Anim Input Adjusted Aim Structure) {name=InPoseUpperBody}
   Output Pose() {layerGroup=Item Override}
  ]}
 ItemZipLineStartAdditive {nodes [Output Pose() {layerGroup=Item Override}]}
 OnBecomeRelevant_Fall {nodes [
   OnBecomeRelevant_Fall!() -> (then, Context:"Anim Update Context Structure (by ref)", Node:"Anim Node Reference Structure (by ref)") {is=title K2Node_FunctionEntry}
   > Set FallTime(FallTime:float=0) -> (Output_Get:float)
   > Return Node()
  ]}
 OnLeapBecomeRelevant {nodes [
   OnLeapBecomeRelevant!() -> (then, Context:"Anim Update Context Structure (by ref)", Node:"Anim Node Reference Structure (by ref)") {is=title K2Node_FunctionEntry}
   > Set InitialWorldSpaceDodgeDirection(InitialWorldSpaceDodgeDirection:Vector=Property Access!() -> (Value:Vector=~) {is=title K2Node_PropertyAccess}) -> (Output_Get:Vector)
   > Set AirborneFromLeap(AirborneFromLeap:bool=true) -> (Output_Get:bool)
   > Return Node()
  ]}
 OnStateExit_Fall {nodes [
   OnStateExit_Fall!() -> (then, Context:"Anim Update Context Structure (by ref)", Node:"Anim Node Reference Structure (by ref)") {is=title K2Node_FunctionEntry}
   > Set AirborneFromLeap(AirborneFromLeap=false)
   > Return Node()
  ]}
 OnUpdate_Fall {nodes [
   OnUpdate_Fall!() -> (then, Context:"Anim Update Context Structure (by ref)", Node:"Anim Node Reference Structure (by ref)") {is=title K2Node_FunctionEntry}
   > Set FallTime(FallTime=float + float(A=Get FallTime() -> (FallTime:float), B=Get DeltaTime() -> (DeltaTime:float)))
   > Return Node()
  ]}
]
variables [
 AirborneFromLeap {type=boolean}
 HoldShieldActive {type=boolean}
]
