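# Translucent unlit GravWell material for Niagara ribbons and mesh particles, drawing a blue audio-spectrum-style trail.
# Three trail-texture channels, waved by paired panning sines and Voronoi noise, form color and opacity; Low and Medium quality share one sample and skip depth fade.
# Particle color, distorted edge fades and sine banding provide further shaping.
# Dynamic parameter index 1 drives pan speed, wave scale and optional SplineThicken world-position-offset width.
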Material /GravWellGameplay/Assets/FX/M_GravWell_AudioTrail_V3
graph {
 properties {
  bAutomaticallySetUsageInEditor=false
  bUsedWithNiagaraMeshParticles=true
  bUsedWithNiagaraRibbons=true
  blendMode=BLEND_Translucent
  shadingModel=MSM_Unlit
  twoSided=true
  usedShadingModels=Unlit
 }
 nodes [
  MaterialFunctionCall(Emissive=Multiply(A=ParticleColor().RGB, B=NamedRerouteUsage())) {materialFunction=&MF_FX_EmissiveAdjustments}
  n1 = NamedRerouteUsage()
  n2 = NamedRerouteUsage()
  n3 = DepthFade(FadeDistance=ScalarParameter() {defaultValue=50, group=03 - Opacity, parameterName=DepthFadeDist})
  n4 = Constant() {r=1}
  Multiply(A=Multiply(A=Multiply(A=@n1, B=@n2), B=ParticleColor().A), B=QualitySwitch(Default=@n3, Low=@n4, High=@n3, Medium=@n4, Epic=@n3))
  n5 = DynamicParameter() {parameterIndex=1, paramNames [SplineThicknessMult, PanningSpeed, TexScaling, WaveScale]}
  n6 = Multiply(A=@n5.SplineThicknessMult, B=ScalarParameter() {group=07- WPO, parameterName=SplineWidthBase})
  n7 = Multiply(A=@n5.SplineThicknessMult, B=ScalarParameter() {group=07- WPO, parameterName=SplineWidthTip})
  n8 = TextureCoordinate()
  n9 = StaticBool()
  NamedRerouteDeclaration(None=StaticSwitchParameter(
    True=MaterialFunctionCall(WidthBase=@n6, WidthTip=@n7, UVs for Projection=@n8, UVs for Thickness=@n8, UVs for texturing=@n8, Expand U or V UV Channel=@n9) {materialFunction=&SplineThicken}.WorldPositionOffset
    False=Constant()
   ) {group=07- WPO, parameterName=UseSplineThicken}) {name=WPOOut, nodeColor=<rgba 1 0.246 0 1>}
  n10 = TextureCoordinate()
  n11 = DynamicParameter() {parameterIndex=1, paramNames [SplineThicknessMult, PanningSpeed, TexScaling, WaveScale]}
  n12 = ScalarParameter() {defaultValue=-2, group=03 - Opacity, parameterName=BandingSpeed}
  n13 = Add(
   A=Multiply(A=ComponentMask(None=@n10) {r=true}, B=ScalarParameter() {defaultValue=100, group=03 - Opacity, parameterName=BandingFreq})
   B=Multiply(A=Multiply(A=@n11.PanningSpeed, B=@n12), B=Time())
  )
  NamedRerouteDeclaration(None=ConstantBiasScale(None=Sine(None=@n13))) {name=BandingAlpha, nodeColor=<rgba 0.0312 0 1 1>}
  n14 = ScalarParameter() {defaultValue=0.25, group=03 - Opacity, parameterName=TopBotFade}
  n15 = Normalize(VectorInput=Add(A=TextureCoordinate()) {constB=-0.5})
  n16 = NamedRerouteUsage()
  n17 = Add(
   A=TextureCoordinate()
   B=Multiply(A=Multiply(A=@n15, B=@n16), B=ScalarParameter() {defaultValue=0.1, group=03 - Opacity, parameterName=EdgeDistoStrength})
  )
  n18 = MaterialFunctionCall(Float2=@n17) {materialFunction=&BreakOutFloat2Components}
  n19 = MaterialFunctionCall(Float2=OneMinus(None=@n17)) {materialFunction=&BreakOutFloat2Components}
  n20 = Multiply(A=@n18.R, B=@n19.R)
  n21 = ScalarParameter() {defaultValue=0.5, group=03 - Opacity, parameterName=SideFade}
  n22 = Multiply(A=@n18.G, B=@n19.G)
  NamedRerouteDeclaration(None=Multiply(A=SmoothStep(Max=@n14, Value=@n20) {constMax=0.05}, B=SmoothStep(Max=@n21, Value=@n22) {constMax=0.05})) {name=EdgeFade, nodeColor=<rgba 0.8197 0.1743 1 1>}
  n23 = VectorParameter() {group=01 - Color, parameterName=BlueThirdColor, sortPriority=2, defaultValue=<rgba 0.5614 0 1 1>}
  n24 = VectorParameter() {group=01 - Color, parameterName=BlueSecondColor, sortPriority=1, defaultValue=<rgba 0.0119 0 1.25 1>}
  n25 = TextureCoordinate()
  n26 = ComponentMask(None=TextureCoordinate()) {r=true}
  n27 = VectorParameter() {
   group=00 - UVs
   parameterName=AudioSpectrumUVs_1
   sortPriority=0
   channelNames {a=Pan2, b=Pan1, g=Scale2, r=Scale1}
   defaultValue=<rgba 1 3.2004 -1.1091 -1.5>
  }
  n28 = DynamicParameter() {parameterIndex=1, paramNames [SplineThicknessMult, PanningSpeed, TexScaling, WaveScale]}
  n29 = MaterialFunctionCall(X=@n28.PanningSpeed, Y=@n28.PanningSpeed, Z=@n28.PanningSpeed, A=@n28.PanningSpeed) {materialFunction=&MakeFloat4}
  n30 = MaterialFunctionCall(Float4=Multiply(
    A=MaterialFunctionCall(X=@n27.Scale1, Y=@n27.Scale1, Z=@n27.Scale1, A=@n27.Scale1) {materialFunction=&MakeFloat4}.Result
    B=@n29.Result
   )) {materialFunction=&BreakOutFloat4Components}
  n31 = Time()
  n32 = Sine(None=Add(A=Multiply(A=@n26, B=@n30.R), B=Multiply(A=@n31, B=@n30.B)))
  n33 = Sine(None=Add(A=Multiply(A=@n26, B=@n30.G), B=Multiply(A=@n31, B=@n30.A)))
  n34 = DynamicParameter() {parameterIndex=1, paramNames [SplineThicknessMult, PanningSpeed, TexScaling, WaveScale]}
  n35 = Multiply(
   A=MaterialFunctionCall(Y=Multiply(A=@n32, B=@n33)) {materialFunction=&MakeFloat2}.Result
   B=Multiply(A=ScalarParameter() {defaultValue=0.2, group=00 - UVs, parameterName=WaveStrength_01}, B=@n34.WaveScale)
  )
  n36 = Multiply(A=NamedRerouteUsage(), B=ScalarParameter() {defaultValue=0.5, group=02 - Ondulation, parameterName=NoiseDistoStrength, sortPriority=5})
  n37 = TextureSample(UVs=Add(A=Add(A=@n25, B=@n35), B=@n36)) {texture=&/T_GravWell_AudioTrail_V2}
  n38 = ComponentMask(None=TextureCoordinate()) {r=true}
  n39 = VectorParameter() {
   group=00 - UVs
   parameterName=AudioSpectrumUVs_2
   sortPriority=0
   channelNames {a=Pan2, b=Pan1, g=Scale2, r=Scale1}
   defaultValue=<rgba 1 3.3651 -0.8 -1>
  }
  n40 = MaterialFunctionCall(Float4=Multiply(
    A=MaterialFunctionCall(X=@n39.Scale1, Y=@n39.Scale1, Z=@n39.Scale1, A=@n39.Scale1) {materialFunction=&MakeFloat4}.Result
    B=@n29.Result
   )) {materialFunction=&BreakOutFloat4Components}
  n41 = Time()
  n42 = Sine(None=Add(A=Multiply(A=@n38, B=@n40.R), B=Multiply(A=@n41, B=@n40.B)))
  n43 = Sine(None=Add(A=Multiply(A=@n38, B=@n40.G), B=Multiply(A=@n41, B=@n40.A)))
  n44 = Multiply(
   A=MaterialFunctionCall(Y=Multiply(A=@n42, B=@n43)) {materialFunction=&MakeFloat2}.Result
   B=ScalarParameter() {defaultValue=0.2, group=00 - UVs, parameterName=WaveStrength_02}
  )
  n45 = Add(A=Add(A=TextureCoordinate(), B=Multiply(A=@n44, B=@n34.WaveScale)), B=@n36)
  n46 = ComponentMask(None=TextureCoordinate()) {r=true}
  n47 = VectorParameter() {
   group=00 - UVs
   parameterName=AudioSpectrumUVs_3
   sortPriority=0
   channelNames {a=Pan2, b=Pan1, g=Scale2, r=Scale1}
   defaultValue=<rgba 1 2 -0.2 -0.5>
  }
  n48 = MaterialFunctionCall(Float4=Multiply(
    A=MaterialFunctionCall(X=@n47.Scale1, Y=@n47.Scale1, Z=@n47.Scale1, A=@n47.Scale1) {materialFunction=&MakeFloat4}.Result
    B=@n29.Result
   )) {materialFunction=&BreakOutFloat4Components}
  n49 = Time()
  n50 = Add(A=Multiply(A=@n46, B=@n48.R), B=Multiply(A=@n49, B=@n48.B))
  n51 = Add(A=Multiply(A=@n46, B=@n48.G), B=Multiply(A=@n49, B=@n48.A))
  n52 = MaterialFunctionCall(Y=Multiply(A=Sine(None=@n50), B=Sine(None=@n51))) {materialFunction=&MakeFloat2}
  n53 = Multiply(A=@n34.WaveScale, B=ScalarParameter() {defaultValue=0.15, group=00 - UVs, parameterName=WaveStrength_03})
  n54 = Add(A=@n36, B=Add(A=TextureCoordinate(), B=Multiply(A=@n52.Result, B=@n53)))
  n55 = MaterialFunctionCall(
   X=@n37.R
   Y=ChannelMaskParameter(None=TextureSample(UVs=@n45) {texture=&T_GravWell_AudioTrail_V2}.RGB) {group=01 - Color, maskChannel=Green, parameterName=SecondTrailMask, defaultValue=<rgba 0 1 0 0>}.R
   Z=ChannelMaskParameter(None=TextureSample(UVs=@n54) {texture=&T_GravWell_AudioTrail_V2}.RGB) {group=01 - Color, maskChannel=Blue, parameterName=ThirdTrailMask, defaultValue=<rgba 0 0 1 0>}.R
  ) {materialFunction=&MakeFloat3}
  n56 = MaterialFunctionCall(
   X=ChannelMaskParameter(None=@n37.RGB) {group=01 - Color, parameterName=FirstTrailMask}.R
   Y=ChannelMaskParameter(None=@n37.RGB) {group=01 - Color, maskChannel=Green, parameterName=SecondTrailMask, defaultValue=<rgba 0 1 0 0>}.R
   Z=ChannelMaskParameter(None=@n37.RGB) {group=01 - Color, maskChannel=Blue, parameterName=ThirdTrailMask, defaultValue=<rgba 0 0 1 0>}.R
  ) {materialFunction=&MakeFloat3}
  n57 = MaterialFunctionCall(Float3=Multiply(
    A=QualitySwitch(Default=@n55.Result, Low=@n56.Result, High=@n55.Result, Medium=@n56.Result, Epic=@n55.Result)
    B=VectorParameter() {group=01 - Color, parameterName=TrailsOpacity, defaultValue=<rgba 0.8 1 2 0>}.RGB
   )) {materialFunction=&BreakOutFloat3Components}
  NamedRerouteDeclaration(None=LinearInterpolate(
    A=LinearInterpolate(A=@n23.RGB, B=@n24.RGB, Alpha=@n57.R)
    B=VectorParameter() {group=01 - Color, parameterName=BlueMainColor, sortPriority=0, defaultValue=<rgba 0 2.8773 10 1>}.RGB
    Alpha=@n57.G
   )) {name=AudioColor, nodeColor=<rgba 0 1 0.6484 1>}
  n58 = Add(A=@n57.B, B=Add(A=@n57.R, B=@n57.R))
  NamedRerouteDeclaration(None=Multiply(A=Saturate(None=@n58), B=NamedRerouteUsage())) {name=AudioSpectrumOpacityOut, nodeColor=<rgba 0 0.9531 1 1>}
  n59 = TextureCoordinate()
  n60 = VectorParameter() {group=02 - Ondulation, parameterName=MacroNoiseUVs, channelNames {a=PanY, b=PanX, g=ScaleX, r=ScaleX}, defaultValue=<rgba 2 0.1 0 2>}
  n61 = MaterialFunctionCall(X=@n60.ScaleX, Y=@n60.ScaleX) {materialFunction=&MakeFloat2}
  n62 = MaterialFunctionCall(X=@n60.PanX, Y=@n60.PanX) {materialFunction=&MakeFloat2}
  n63 = Multiply(A=Time(), B=ScalarParameter() {defaultValue=-0.1, group=02 - Ondulation, parameterName=OndulationSpeed}) {constB=-0.1}
  n64 = MaterialFunctionCall(X=ParticleRandom()) {materialFunction=&MakeFloat2}
  n65 = Add(A=Add(A=Multiply(A=@n59, B=@n61.Result), B=Multiply(A=@n62.Result, B=@n63)), B=@n64.Result)
  n66 = TextureCoordinate()
  n67 = VectorParameter() {group=02 - Ondulation, parameterName=MicroNoiseUVs, channelNames {a=PanY, b=PanX, g=ScaleX, r=ScaleX}, defaultValue=<rgba 4 0.1 0 2>}
  n68 = MaterialFunctionCall(X=@n67.ScaleX, Y=@n67.ScaleX) {materialFunction=&MakeFloat2}
  n69 = MaterialFunctionCall(X=@n67.PanX, Y=@n67.PanX) {materialFunction=&MakeFloat2}
  n70 = Add(A=@n64.Result, B=Add(A=Multiply(A=@n66, B=@n68.Result), B=Multiply(A=@n69.Result, B=@n63)))
  n71 = Multiply(
   A=SmoothStep(
    Min=ScalarParameter() {group=02 - Ondulation, parameterName=OndulationTreshold}
    Value=TextureSample(UVs=@n65) {texture=/Game/Effects/Fort_Effects/Textures/T_PinkSpike_Voronoi2D_Height}.RGB
   ) {constMin=0.5}
   B=OneMinus(None=TextureSample(UVs=@n70) {texture=&T_PinkSpike_Voronoi2D_Height}.RGB)
  )
  NamedRerouteDeclaration(None=ComponentMask(None=SmoothStep(Value=@n71)) {r=true}) {name=Distortion, nodeColor=<rgba 0.4766 1 0 1>}
  NamedRerouteUsage()
 ]
}
