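# Translucent unlit GravWell material for a soft shockwave on Niagara mesh particles, with emissive driven by particle color despite the "Textured" name.
# A time-scrolled sine along V modulates index of refraction toward a dynamic parameter, masked by vertical gradient falloff and inverted Fresnel.
# Opacity combines particle alpha, the gradient mask and a brightness scalar with particle-controlled depth fade.
# A second depth fade applies except on mobile.
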Material /GravWellGameplay/Assets/FX/M_GravWell_SoftShockwave_Textured
graph {
 properties {
  bUsedWithNanite=true
  bUsedWithNiagaraMeshParticles=true
  blendMode=BLEND_Translucent
  pixelDepthOffsetMode=PDOM_Legacy
  previewMesh=&/SM_GravWell_ShieldVolume
  refractionMethod=RM_IndexOfRefraction
  shadingModel=MSM_Unlit
  usedShadingModels=Unlit
 }
 nodes [
  n1 = ComponentMask(None=TextureCoordinate()) {g=true}
  n2 = Multiply(A=ScalarParameter() {defaultValue=-0.5, parameterName=WaveSpeed}, B=Time())
  n3 = Sine(None=Multiply(A=Add(A=@n1, B=@n2), B=ScalarParameter() {defaultValue=5, parameterName=WaveNumber}))
  n4 = MaterialFunctionCall() {materialFunction=&LinearGradient}
  n5 = Multiply(A=Multiply(A=OneMinus(None=@n4.VGradient), B=@n4.VGradient)) {constB=2}
  n6 = ScalarParameter() {defaultValue=0.5, parameterName=BottomFalloff}
  n7 = Saturate(None=Power(Base=Divide(A=@n5, B=@n6)))
  n8 = MaterialFunctionCall(In=OneMinus(None=Fresnel() {exponent=1}), Contrast=Constant() {r=1}) {materialFunction=&CheapContrast}
  LinearInterpolate(
   B=LinearInterpolate(
    A=DynamicParameter() {defaultValue=<rgba 1 1 1 100>, paramNames [RefractionIndex, Erode, ErodeHardness, DepthFadeAmount]}.RefractionIndex
    Alpha=ConstantBiasScale(None=@n3)
   )
   Alpha=Multiply(A=@n7, B=@n8.Result)
  ) {constA=1}
  n9 = ParticleColor()
  MaterialFunctionCall(Emissive=@n9.RGB) {materialFunction=&MF_FX_EmissiveAdjustments}
  Multiply(
   A=MaterialFunctionCall(In=Fresnel() {exponent=1}, Contrast=Constant() {r=0.1}) {materialFunction=&CheapContrast}.Result
   B=@n8.Result
  )
  n10 = Clamp(None=Multiply(B=ScalarParameter() {defaultValue=1, parameterName=FresnelBrightness}) {constA=1})
  n11 = DynamicParameter() {defaultValue=<rgba 1 1 1 100>, paramNames [RefractionIndex, Erode, ErodeHardness, DepthFadeAmount]}
  n12 = Multiply(A=@n9.A, B=Multiply(A=@n7, B=DepthFade(Opacity=@n10, FadeDistance=@n11.DepthFadeAmount)))
  ShadingPathSwitch(Default=DepthFade(Opacity=@n12, FadeDistance=ScalarParameter() {defaultValue=500, parameterName=FadeDistance}), Mobile=@n12) {desc=Migrated from FeatureLevelSwitch}
 ]
}
