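# Additive unlit Niagara sprite and mesh particle material for the GravWell sound refraction effect, faking rippling waves through UV distortion rather than true refraction.
# Two panning noise layers, offset by particle random, multiply into a Distortion mask that warps a Voronoi texture.
# That texture blends two colors tinted by particle color and drives opacity.
# Inverted Fresnel, particle alpha and a distorted UV edge fade shape the opacity.
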Material /GravWellGameplay/Assets/FX/M_GravWell_SoundRefraction
graph {
 properties {
  bAutomaticallySetUsageInEditor=false
  bUsedWithNiagaraMeshParticles=true
  bUsedWithNiagaraSprites=true
  blendMode=BLEND_Additive
  shadingModel=MSM_Unlit
  usedShadingModels=Unlit
 }
 nodes [
  LinearInterpolate()
  n1 = VectorParameter() {parameterName=Color_1, defaultValue=<rgba 0.2277 0.3735 1 1>}
  n2 = VectorParameter() {parameterName=Color_2, defaultValue=<rgba 0.4751 0.3797 1 1>}
  n3 = ScalarParameter() {group=03 - Opacity, parameterName=OpacityThreshold}
  n4 = VectorParameter() {group=02 - Ondulation, parameterName=ColorNoiseUVs, channelNames {a=PanY, b=PanX, g=ScaleX, r=ScaleX}, defaultValue=<rgba 1 1 0 2>}
  n5 = Multiply(
   A=TextureCoordinate()
   B=MaterialFunctionCall(X=@n4.ScaleX, Y=@n4.ScaleX) {materialFunction=&MakeFloat2}.Result
  )
  n6 = Multiply(A=MaterialFunctionCall(X=@n4.PanX, Y=@n4.PanX) {materialFunction=&MakeFloat2}.Result)
  n7 = NamedRerouteUsage()
  n8 = ScalarParameter() {defaultValue=0.75, parameterName=DistoStrength}
  n9 = TextureSample(UVs=Add(A=Add(A=@n5, B=@n6), B=Multiply(A=@n7, B=@n8))) {texture=/Game/Effects/Fort_Effects/Textures/T_PinkSpike_Voronoi2D_Height}
  n10 = SmoothStep(Min=Add(A=@n3) {constB=0.25}, Value=@n9.RGB) {constMin=0.25}
  n11 = ParticleColor()
  MaterialFunctionCall(Emissive=Multiply(A=LinearInterpolate(A=@n1.RGB, B=@n2.RGB, Alpha=@n10), B=@n11.RGB)) {materialFunction=&MF_FX_EmissiveAdjustments}
  n12 = ScalarParameter() {group=03 - Opacity, parameterName=MinOpacity}
  n13 = SmoothStep(Min=@n3, Value=@n9.RGB)
  n14 = ScalarParameter() {defaultValue=0.2, group=03 - Opacity, parameterName=FresnelExpo}
  n15 = ScalarParameter() {group=03 - Opacity, parameterName=FresnelFrac}
  n16 = Power(Base=OneMinus(None=Fresnel(ExponentIn=@n14, BaseReflectFractionIn=@n15)))
  Multiply(A=Multiply(A=LinearInterpolate(A=@n12, Alpha=@n13), B=Multiply(A=@n11.A, B=@n16)), B=NamedRerouteUsage())
  n17 = ScalarParameter() {group=03 - Opacity, parameterName=TopBotFade}
  n18 = Normalize(VectorInput=Add(A=TextureCoordinate()) {constB=-0.5})
  n19 = NamedRerouteUsage()
  n20 = Add(A=TextureCoordinate(), B=Multiply(A=Multiply(A=@n18, B=@n19), B=ScalarParameter() {group=03 - Opacity, parameterName=EdgeDistoStrength}))
  n21 = MaterialFunctionCall(Float2=@n20) {materialFunction=&BreakOutFloat2Components}
  n22 = MaterialFunctionCall(Float2=OneMinus(None=@n20)) {materialFunction=&BreakOutFloat2Components}
  n23 = Multiply(A=@n21.R, B=@n22.R)
  n24 = ScalarParameter() {group=03 - Opacity, parameterName=SideFade}
  n25 = Multiply(A=@n21.G, B=@n22.G)
  NamedRerouteDeclaration(None=Multiply(A=SmoothStep(Max=@n17, Value=@n23) {constMax=0.05}, B=SmoothStep(Max=@n24, Value=@n25) {constMax=0.05})) {name=EdgeFade, nodeColor=<rgba 0.8197 0.1743 1 1>}
  n26 = VectorParameter() {group=02 - Ondulation, parameterName=MacroNoiseUVs, channelNames {a=PanY, b=PanX, g=ScaleX, r=ScaleX}, defaultValue=<rgba 1 2 0 -1>}
  n27 = Multiply(A=Time(), B=ScalarParameter() {defaultValue=-0.1, group=02 - Ondulation, parameterName=OndulationSpeed}) {constB=-0.1}
  n28 = Add(
   A=Multiply(
    A=TextureCoordinate()
    B=MaterialFunctionCall(X=@n26.ScaleX, Y=@n26.ScaleX) {materialFunction=&MakeFloat2}.Result
   )
   B=Multiply(
    A=MaterialFunctionCall(X=@n26.PanX, Y=@n26.PanX) {materialFunction=&MakeFloat2}.Result
    B=@n27
   )
  )
  n29 = MaterialFunctionCall(X=ParticleRandom()) {materialFunction=&MakeFloat2}
  n30 = TextureCoordinate()
  n31 = VectorParameter() {group=02 - Ondulation, parameterName=MicroNoiseUVs, channelNames {a=PanY, b=PanX, g=ScaleX, r=ScaleX}, defaultValue=<rgba 0.5 4 0 2>}
  n32 = MaterialFunctionCall(X=@n31.ScaleX, Y=@n31.ScaleX) {materialFunction=&MakeFloat2}
  n33 = MaterialFunctionCall(X=@n31.PanX, Y=@n31.PanX) {materialFunction=&MakeFloat2}
  n34 = Add(A=@n29.Result, B=Add(A=Multiply(A=@n30, B=@n32.Result), B=Multiply(A=@n33.Result, B=@n27)))
  n35 = Multiply(
   A=TextureSample(UVs=Add(A=@n28, B=@n29.Result)) {samplerType=SAMPLERTYPE_LinearGrayscale, texture=/Game/Effects/Fort_Effects/Textures/T_DarkAndLight_Noise}.RGB
   B=OneMinus(None=TextureSample(UVs=@n34) {samplerType=SAMPLERTYPE_LinearGrayscale, texture=&T_DarkAndLight_Noise}.RGB)
  )
  NamedRerouteDeclaration(None=ComponentMask(None=SmoothStep(Value=@n35)) {r=true}) {name=Distortion, nodeColor=<rgba 0.4766 1 0 1>}
 ]
}
