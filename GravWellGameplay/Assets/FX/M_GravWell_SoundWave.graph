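# Translucent unlit Niagara material for the GravWell sound wave effect, drawing repeating rings with undulating edges and 2D-offset refraction.
# Radial or straight UVs, dynamic ring parameters and two panned, per-particle-offset noise samples shape an edge-masked line pattern.
# Color blends between blue and pink gradients via BluePink, tinted by particle color.
# Particle alpha and an edge falloff fade the rings; depth fade and refraction strength are quality-switched for cheaper platforms.
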Material /GravWellGameplay/Assets/FX/M_GravWell_SoundWave
graph {
 properties {
  bAutomaticallySetUsageInEditor=false
  bUsedWithNiagaraMeshParticles=true
  bUsedWithNiagaraRibbons=true
  bUsedWithNiagaraSprites=true
  blendMode=BLEND_Translucent
  refractionMethod=RM_2DOffset
  shadingModel=MSM_Unlit
  twoSided=true
  usedShadingModels=Unlit
 }
 nodes [
  n1 = DynamicParameter() {defaultValue=<rgba 0 1 0.2 0>, paramNames [RingProgress, RingNumber, RingThickness, BluePink]}
  n2 = Floor(None=Multiply(A=ComponentMask() {g=true}, B=@n1.RingNumber))
  n3 = Constant() {r=1}
  Add(B=Multiply(
    A=Divide(A=@n2, B=@n1.RingNumber)
    B=MaterialFunctionCall(Y=@n3) {materialFunction=&MakeFloat2}.Result
   ) {constB=0.1})
  n4 = NamedRerouteUsage()
  n5 = ScalarParameter() {defaultValue=1, parameterName=ColorFreq}
  n6 = Frac(None=ComponentMask(None=Multiply(A=@n4, B=@n5)) {r=true})
  n7 = LinearInterpolate(
   A=VectorParameter() {group=01 - Color, parameterName=BaseMainColor, sortPriority=2, defaultValue=<rgba 0.0123 0 1 1>}.RGB
   B=VectorParameter() {group=01 - Color, parameterName=BaseBoost, sortPriority=1, defaultValue=<rgba 0 0.3057 1 1>}.RGB
   Alpha=@n6
  )
  n8 = VectorParameter() {group=01 - Color, parameterName=PinkBaseColor, sortPriority=6, defaultValue=<rgba 0.1248 0 0.5469 1>}
  n9 = LinearInterpolate(
   A=@n8.RGB
   B=VectorParameter() {group=01 - Color, parameterName=PinkBoostColor, sortPriority=5, defaultValue=<rgba 0.9123 0 1 1>}.RGB
   Alpha=@n6
  )
  n10 = DynamicParameter() {defaultValue=<rgba 0 1 0.2 0>, paramNames [RingProgress, RingNumber, RingThickness, BluePink]}
  n11 = ParticleColor()
  MaterialFunctionCall(Emissive=Multiply(A=LinearInterpolate(A=@n7, B=@n9, Alpha=@n10.BluePink), B=@n11.RGB)) {materialFunction=&MF_FX_EmissiveAdjustments}
  n12 = Constant() {r=1}
  MaterialFunctionCall(X=@n12, Y=DynamicParameter() {defaultValue=<rgba 0 1 0.2 0>, paramNames [RingProgress, RingNumber, RingThickness, BluePink]}.RingNumber) {materialFunction=&MakeFloat2}
  n13 = Add(A=ComponentMask(None=NamedRerouteUsage()) {g=true}) {constB=0.25}
  n14 = DynamicParameter() {defaultValue=<rgba 0 1 0.2 0>, paramNames [RingProgress, RingNumber, RingThickness, BluePink]}
  Multiply(A=@n13, B=@n14.RingNumber)
  n15 = Multiply(
   A=ComponentMask(None=NamedRerouteUsage()) {g=true}
   B=LinearInterpolate(Alpha=ScalarParameter() {defaultValue=1, parameterName=UseRadialMapping}) {constA=1, constB=2}
  ) {constB=2}
  n16 = OneMinus(None=@n15)
  n17 = Saturate(None=Divide(A=Multiply(A=@n15, B=@n16)) {constB=0.25})
  n18 = Constant()
  n19 = ScalarParameter() {group=03 - Opacity, parameterName=RefractionStrength}
  n20 = MaterialFunctionCall!(
   Low _ PC Low and Med=@n18
   "High _ XB1, PS4"=@n19
   "Epic _ PC High, XSX, PS5"=@n19
   --------
   "Mobile Low _ PC Perf, Android"=@n18
   Mobile Med _ Switch=@n18
  ) {materialFunction=/Game/Global/MaterialFunctions/TechArt/MF_Fortnite_QualitySwitch_Scalar}
  n21 = TextureCoordinate()
  Multiply(A=Multiply(A=@n17, B=Multiply(A=@n11.A, B=@n20.Result)), B=Normalize(VectorInput=Add(A=@n21) {constB=-0.5}))
  n22 = NamedRerouteUsage()
  n23 = ScalarParameter() {defaultValue=-0.5, parameterName=OndulationStrength}
  n24 = Constant()
  n25 = Add(
   A=@n13
   B=Multiply(
    A=ComponentMask(None=@n22) {g=true, r=true}
    B=MaterialFunctionCall(X=@n23, Y=@n24) {materialFunction=&MakeFloat2}.Result
   )
  )
  n26 = VectorParameter() {parameterName=LineRefine, channelNames {a="", b=Power, g=Multiply, r=Offset}, defaultValue=<rgba 0 2 30 0>}
  n27 = Multiply(
   A=MaterialFunctionCall!(UorV=Frac(None=@n25), EdgeControl=@n26) {materialFunction=/Game/Effects/Niagara/MasterMaterials/Functions/Common/FXCommon_EdgeMask_Simple}.Mask
   B=@n17
  )
  n28 = DepthFade(FadeDistance=ScalarParameter() {group=03 - Opacity, parameterName=DepthFadeDist})
  n29 = Constant() {r=1}
  Multiply(A=Saturate(None=Multiply(A=@n27, B=@n11.A)), B=QualitySwitch(Default=@n28, Low=@n29, High=@n28, Medium=@n29, Epic=@n28))
  n30 = Multiply(A=@n17, B=Subtract(A=@n15) {constB=0.9})
  n31 = Add(
   A=ComponentMask(None=NamedRerouteUsage()) {r=true}
   B=Multiply(A=ScalarParameter() {defaultValue=0.2, group=04- EdgeControl, parameterName=EdgeColorSpeed}, B=Time())
  )
  n32 = ScalarParameter() {defaultValue=10, group=04- EdgeControl, parameterName=EdgeColorFreq}
  n33 = Saturate(None=Sine(None=Multiply(A=@n31, B=@n32) {constB=5}))
  Multiply(A=Saturate(None=Divide(A=@n30) {constB=0.01}), B=Saturate(None=Divide(A=@n33) {constB=0.01}))
  n34 = TextureCoordinate()
  n35 = MaterialFunctionCall(XY=@n34) {materialFunction=&Swizzle}
  n36 = ScalarParameter() {parameterName=SwizzleUVs}
  n37 = Add(A=@n34) {constB=-0.5}
  NamedRerouteDeclaration(None=LinearInterpolate(
    A=LinearInterpolate(A=@n34, B=@n35.YX, Alpha=@n36)
    B=MaterialFunctionCall(Vector or UVs=@n37) {materialFunction=&VectorToRadialValue}.Radial Coordinates
    Alpha=ScalarParameter() {defaultValue=1, parameterName=UseRadialMapping}
   )) {name=BaseUVs, nodeColor=<rgba 1 0 0.6328 1>}
  NamedRerouteDeclaration(None=@n14.RingProgress) {name=CustomTime, nodeColor=<rgba 0.1953 0 1 1>}
  n38 = NamedRerouteUsage()
  n39 = VectorParameter() {parameterName=MacroNoiseUVs, channelNames {a=PanY, b=PanX, g=ScaleX, r=ScaleX}, defaultValue=<rgba 1 0.1 0 0.25>}
  n40 = Multiply(A=Time(), B=ScalarParameter() {defaultValue=-0.1, parameterName=OndulationSpeed}) {constB=-0.1}
  n41 = Add(
   A=Multiply(
    A=Multiply(A=@n12, B=@n38)
    B=MaterialFunctionCall(X=@n39.ScaleX, Y=@n39.ScaleX) {materialFunction=&MakeFloat2}.Result
   )
   B=Multiply(
    A=MaterialFunctionCall(X=@n39.PanX, Y=@n39.PanX) {materialFunction=&MakeFloat2}.Result
    B=@n40
   )
  )
  n42 = MaterialFunctionCall(X=ParticleRandom()) {materialFunction=&MakeFloat2}
  n43 = TextureObjectParameter() {
   parameterName=DistortionTexture
   samplerType=SAMPLERTYPE_LinearGrayscale
   texture=/Game/Effects/Fort_Effects/Textures/Tile/T_CellNoise_Tile_M
  }
  n44 = SmoothStep(
   Min=ScalarParameter() {parameterName=OndulationTreshold}
   Value=TextureSample(UVs=Add(A=@n41, B=@n42.Result), Tex=@n43) {texture=/Game/Effects/Fort_Effects/Textures/T_PinkSpike_Voronoi2D_Height}.RGB
  ) {constMin=0.5}
  n45 = VectorParameter() {parameterName=MicroNoiseUVs, channelNames {a=PanY, b=PanX, g=ScaleX, r=ScaleX}, defaultValue=<rgba 2 0.1 0 0.5>}
  n46 = Multiply(
   A=Multiply(A=@n12, B=NamedRerouteUsage())
   B=MaterialFunctionCall(X=@n45.ScaleX, Y=@n45.ScaleX) {materialFunction=&MakeFloat2}.Result
  )
  n47 = Multiply(
   A=MaterialFunctionCall(X=@n45.PanX, Y=@n45.PanX) {materialFunction=&MakeFloat2}.Result
   B=@n40
  )
  n48 = TextureSample(UVs=Add(A=@n42.Result, B=Add(A=@n46, B=@n47)), Tex=@n43) {texture=&T_PinkSpike_Voronoi2D_Height}
  NamedRerouteDeclaration(None=SmoothStep(Value=Multiply(A=@n44, B=@n48.RGB))) {name=NoiseAlphaValue, nodeColor=<rgba 0.3564 0.474 1 1>}
  NamedRerouteUsage()
 ]
}
