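# Material function for a gauntlet equip effect that dissolves the mesh into view with a glowing edge.
# Time since custom-primitive-data C_EquipAnim drives a reveal over EquipAnimDuration along a pre-skinned local Z gradient, low to high.
# Panning, cell-quantized hash noise on object-stabilized screen-space UVs breaks up the edge.
# The input material attributes get the resulting opacity mask, plus a two-color sphere-mask edge glow added to emissive.
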MaterialFunction /GravWellGameplay/Assets/Materials/MF_GauntletEquip
graph {nodes [
  Constant() {r=1}
  n1 = NamedRerouteUsage()
  n2 = NamedRerouteUsage()
  FunctionOutput(None=SetMaterialAttributes!(MaterialAttributes=NamedRerouteUsage(), EmissiveColor=Add(A=@n1, B=@n2), OpacityMask=NamedRerouteUsage()) {attributeSetTypes [#1, #2]})
  n3 = GetMaterialAttributes(None=FunctionInput(Preview=SetMaterialAttributes!(MaterialAttributes)) {inputName=Material, inputType=FunctionInput_MaterialAttributes}) {attributeGetTypes [#1]}
  NamedRerouteDeclaration(None=@n3.MaterialAttributes) {name=InputMaterial, nodeColor=<rgba 0.5937 1 0 1>}
  NamedRerouteDeclaration(None=FunctionInput(Preview=Constant3Vector()) {bUsePreviewValueAsDefault=true, inputName=PreskinnedLocalPosition, sortPriority=1}) {name=PreSkinnedLocalPosition, nodeColor=<rgba 1 0 0.9141 1>}
  NamedRerouteDeclaration(None=@n3.EmissiveColor) {name=Input_EmissiveColor, nodeColor=<rgba 1 0 0.7734 1>}
  n4 = Divide(
   A=Add(A=NamedRerouteUsage(), B=VectorParameter() {parameterName=Preskin_Offset_amt, defaultValue=<rgba 0 0 -88 1>}.RGB)
   B=VectorParameter() {parameterName=GradientBound, defaultValue=<rgba 40 40 40 1>}.RGB
  )
  n5 = OneMinus(None=Saturate(None=ComponentMask(None=@n4) {b=true}))
  n6 = Divide(
   A=Subtract(A=Time(), B=ScalarParameter() {bUseCustomPrimitiveData=true, parameterName=C_EquipAnim, primitiveDataIndex=1})
   B=ScalarParameter() {defaultValue=1, parameterName=EquipAnimDuration}
  )
  n7 = Multiply(A=OneMinus(None=Saturate(None=@n6))) {constB=2}
  n8 = MaterialFunctionCall!(Position=ActorPositionWS(), UVRadialScale=ScalarParameter() {defaultValue=0.02, parameterName=ScreenUV_Scale}) {materialFunction=/Game/Weapons/WeaponSkins/MaterialFunction/ObjectStabilizedScreenSpaceUVs}
  n9 = ComponentMask(None=Multiply(A=Time(), B=VectorParameter() {parameterName=Equipnoise_PanSpeed, defaultValue=<rgba 0 2 0 1>}.RGB)) {g=true, r=true}
  n10 = MaterialFunctionCall(Vec2=Ceil(None=Add(A=@n8.StabilizedUVs, B=@n9))) {materialFunction=&MF_Hash}
  n11 = Multiply(A=ComponentMask(None=ConstantBiasScale(None=@n10.Vec2)) {r=true}, B=ScalarParameter() {defaultValue=0.35, parameterName=Reznoise_Strength})
  n12 = Saturate(None=Subtract(A=Add(A=@n5), B=Subtract(A=@n7, B=@n11)))
  n13 = Saturate(None=@n10.Vec2)
  n14 = Multiply(
   A=SphereMask(
    A=@n12
    B=ScalarParameter() {parameterName=GlowSamplePos}
    Radius=ScalarParameter() {defaultValue=0.5, parameterName=EdgeGlowRadius}
    Hardness=ScalarParameter() {parameterName=EdgeGlowHardness}
   )
   B=LinearInterpolate(
    A=VectorParameter() {parameterName=EmissivelColor1, defaultValue=<rgba 0.5376 0 1 1>}.RGB
    B=VectorParameter() {parameterName=EmissivelColor2, defaultValue=<rgba 0 0.3709 1 1>}.RGB
    Alpha=ComponentMask(None=@n13) {r=true}
   )
  )
  NamedRerouteDeclaration(None=Power(Base=Multiply(A=@n14) {constB=5})) {name=EdgeGlow, nodeColor=<rgba 0 1 0.9063 1>}
  NamedRerouteDeclaration(None=@n12) {name=OpacityMask, nodeColor=<rgba 1 0.5859 0 1>}
 ]}
