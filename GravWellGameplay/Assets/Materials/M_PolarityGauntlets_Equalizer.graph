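# Two-sided skeletal-mesh material that draws an animated audio-equalizer style display with shield-like effects on the Polarity Gauntlets.
# UVs quantize into stem and pitch cells with selectable shapes, and hashed noise plus sine waves vary the levels, colored through CA_PulsarShield curve atlas rows.
# Masks add activation edge glow, fresnel, pulse and hit reactions, with hit timing looping on a DebugTime clock.
# Vertex red blends the result in before MF_GauntletEquip.
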Material /GravWellGameplay/Assets/Materials/M_PolarityGauntlets_Equalizer
graph {
 properties {
  bAutomaticallySetUsageInEditor=false
  bUseMaterialAttributes=true
  bUsedWithMeshDeformer=true
  bUsedWithSkeletalMesh=true
  twoSided=true
  usedShadingModels=Default Lit
 }
 nodes [
  LinearInterpolate()
  n1 = Constant3Vector() {constant=<rgba 0 0.3278 1 1>}
  n2 = ScalarParameter() {defaultValue=0.15, parameterName=ShieldBG_Brightness}
  n3 = ScalarParameter() {defaultValue=0.075, parameterName=ShieldBG_Cell_GrooveBrightness}
  n4 = NamedRerouteUsage()
  n5 = ScalarParameter() {defaultValue=0.5, parameterName=EmissiveStrength}
  n6 = NamedRerouteUsage()
  n7 = Divide(A=NamedRerouteUsage())
  n8 = Subtract(A=@n6, B=LinearInterpolate(B=NamedRerouteUsage(), Alpha=Clamp(None=@n7)) {constA=1})
  n9 = Saturate(None=Subtract(A=StaticSwitchParameter(True=@n6, False=@n8) {parameterName=SwitchToBaseColorLEvels}, B=NamedRerouteUsage()))
  n10 = CurveAtlasRowParameter(CurveTime=@n9) {
   atlas=&/CA_PulsarShield
   curve=&/C_PulsarShield_01
   parameterName=ColorCurve_01
  }
  n11 = CurveAtlasRowParameter(CurveTime=@n9) {
   atlas=&CA_PulsarShield
   curve=&/C_PulsarShield_02
   defaultValue=1
   parameterName=ColorCurve_02
  }
  n12 = Floor(None=Divide(A=NamedRerouteUsage()))
  n13 = LinearInterpolate(A=@n10, B=@n11, Alpha=OneMinus(None=Clamp(None=@n12)))
  n14 = LinearInterpolate(
   A=Multiply(A=@n1, B=LinearInterpolate(A=@n2, B=@n3, Alpha=@n4)) {constB=0.15}
   B=Power(Base=Multiply(A=@n5, B=@n13), Exp=ScalarParameter() {defaultValue=1, parameterName=EmissivePower})
   Alpha=Multiply(A=NamedRerouteUsage(), B=Ceil(None=@n8))
  )
  n15 = LinearInterpolate(
   A=Multiply(A=@n14) {constB=0.5}
   B=Multiply(A=@n14, B=ScalarParameter() {defaultValue=1.5, parameterName=FrenselEmissive_Boost})
   Alpha=NamedRerouteUsage()
  )
  n16 = Multiply(A=@n1, B=ScalarParameter() {defaultValue=5, parameterName=Activate_Edgeglow_Strength}) {constB=10}
  n17 = SetMaterialAttributes!(
   MaterialAttributes
   Base Color=@n15
   Metallic=Constant()
   Specular=ScalarParameter() {defaultValue=1, parameterName=Specular}
   Roughness=ScalarParameter() {defaultValue=0.2, parameterName=Roughness}
   Emissive Color=LinearInterpolate(A=@n15, B=Power(Base=@n16), Alpha=NamedRerouteUsage())
  ) {attributeSetTypes [
    #1
    #2
    #3
    #4
    #5
   ]}
  n18 = LocalPosition() {localOrigin=InstancePreSkinning}
  MaterialFunctionCall!(
   Material=BlendMaterialAttributes(A=SetMaterialAttributes!(MaterialAttributes), B=@n17, Alpha=VertexColor().R)
   PreskinnedLocalPosition=ShaderStageSwitch(PixelShader=VertexInterpolator(VS=@n18.XYZ).PS, VertexShader=@n18.XYZ)
  ) {materialFunction=&/MF_GauntletEquip}
  n19 = Multiply(A=TextureCoordinate()) {constB=2}
  n20 = MaterialFunctionCall(Float2=NamedRerouteDeclaration(None=Subtract(A=@n19)) {name=BaseUV_Coord, nodeColor=<rgba 0.5937 1 0 1>}) {materialFunction=&BreakOutFloat2Components}
  NamedRerouteDeclaration(None=Abs(None=NamedRerouteDeclaration(None=@n20.R) {name=XVector, nodeColor=<rgba 0.2656 1 0 1>})) {name=XVector_Abs, nodeColor=<rgba 0 0.5078 1 1>}
  n21 = NamedRerouteDeclaration(None=@n20.G) {name=YVector, nodeColor=<rgba 1 0 0.5859 1>}
  NamedRerouteDeclaration(None=OneMinus(None=Abs(None=@n21))) {name=YVector_Abs, nodeColor=<rgba 0.8281 1 0 1>}
  n22 = Add(
   A=Add(A=NamedRerouteUsage()) {constB=0.25}
   B=Multiply(A=Time(), B=ScalarParameter() {defaultValue=-1, parameterName=WaveLengthPanSpeed}) {constB=-1}
  )
  NamedRerouteDeclaration(None=ConstantBiasScale(None=Sine(None=@n22))) {name=SineWave, nodeColor=<rgba 0.9453 1 0 1>}
  n23 = NamedRerouteUsage()
  n24 = AppendVector(A=@n23, B=@n23)
  n25 = ConstantBiasScale(None=MaterialFunctionCall(Seed=ScalarParameter() {defaultValue=0.77, parameterName=StemSeed}, Vec2=@n24) {materialFunction=&MF_Hash}.Vec2)
  NamedRerouteDeclaration(None=ComponentMask(None=Saturate(None=@n25)) {r=true}) {name=PitchVariationMask, nodeColor=<rgba 0 0.8828 1 1>}
  n26 = ConstantBiasScale(None=MaterialFunctionCall(Seed=ScalarParameter() {defaultValue=0.32, parameterName=StemTimeVariationSpeed}, Vec2=@n24) {materialFunction=&MF_Hash}.Vec2)
  NamedRerouteDeclaration(None=ComponentMask(None=Saturate(None=@n26)) {r=true}) {name=PitchTimeVariation, nodeColor=<rgba 0 1 0.2031 1>}
  n27 = NamedRerouteUsage()
  NamedRerouteDeclaration(None=LinearInterpolate(
    A=ScalarParameter() {defaultValue=0.1, parameterName=ColorVariation_Min}
    B=ScalarParameter() {defaultValue=-0.1, parameterName=ColorVariation_Max}
    Alpha=Saturate(None=@n27)
   ) {constB=-0.2}) {name=ColorVariation_Mask, nodeColor=<rgba 0.8047 1 0 1>}
  n28 = NamedRerouteUsage()
  n29 = Constant()
  n30 = Subtract(A=Multiply(A=OneMinus(None=@n28), B=NamedRerouteUsage()), B=LinearInterpolate(Alpha=Saturate(None=@n29)) {constA=-1.1})
  NamedRerouteDeclaration(None=Saturate(None=Ceil(None=@n30))) {name=Activate/Deactivate, nodeColor=<rgba 0 0.5781 1 1>}
  NamedRerouteDeclaration(None=@n8) {name=Beats, nodeColor=<rgba 0.4766 0 1 1>}
  NamedRerouteDeclaration(None=SphereMask(
    A=@n30
    B=ScalarParameter() {parameterName=Activate_EdgeSample}
    Radius=ScalarParameter() {defaultValue=0.2, parameterName=Activate_EdgeRadius}
    Hardness=ScalarParameter() {parameterName=Activate_EdgeHardneess}
   )) {name=Activate_EdgeMask, nodeColor=<rgba 0.2422 1 0 1>}
  n31 = NamedRerouteDeclaration(None=ScalarParameter() {defaultValue=30, parameterName=StemCound}) {name=StemCount, nodeColor=<rgba 0 0.6484 1 1>}
  n32 = Floor(None=Multiply(A=NamedRerouteUsage(), B=@n31))
  n33 = NamedRerouteDeclaration(None=ScalarParameter() {defaultValue=30, parameterName=PitchCount}) {name=PitchCOunt, nodeColor=<rgba 0.9688 1 0 1>}
  n34 = Floor(None=Multiply(A=NamedRerouteUsage(), B=@n33))
  NamedRerouteDeclaration(None=AppendVector(A=Divide(A=@n32, B=@n31), B=Divide(A=@n34, B=@n33))) {name=PixelUV, nodeColor=<rgba 0.0547 0 1 1>}
  n35 = Multiply(A=NamedRerouteUsage(), B=@n31)
  NamedRerouteDeclaration(None=Divide(A=Floor(None=@n35), B=@n31)) {name=StemGradient, nodeColor=<rgba 0.8516 0 1 1>}
  n36 = NamedRerouteUsage()
  n37 = VectorParameter() {parameterName=Empower_HitLocation, defaultValue=<rgba 0 0 0 1>}
  n38 = ComponentMask(None=@n37.RGB) {g=true, r=true}
  n39 = VectorParameter() {parameterName=BaseUVSCale, defaultValue=<rgba 0.7 1 0 1>}
  n40 = SphereMask(
   A=MaterialFunctionCall(UVs=Subtract(A=@n36, B=@n38), Texture Scale=ComponentMask(None=@n39.RGB) {g=true, r=true}) {materialFunction=&ScaleUVsByCenter}.UVs
   B=Constant2Vector()
   Radius=ScalarParameter() {defaultValue=0.4, parameterName=PowerUpRadius}
   Hardness=ScalarParameter() {parameterName=PowerUp_Hardness}
  )
  n41 = Frac(None=Multiply(A=Time()) {constB=2})
  n42 = SphereMask(
   A=@n40
   B=Multiply(A=ScalarParameter() {defaultValue=1, parameterName=SamplePos}, B=@n41)
   Radius=ScalarParameter() {defaultValue=0.25, parameterName=PulseRadius}
   Hardness=ScalarParameter() {parameterName=PulseHArdness}
  )
  NamedRerouteDeclaration(None=Multiply(A=Multiply(A=@n40, B=@n42), B=NamedRerouteUsage())) {name=PulseGradient, nodeColor=<rgba 0 0.9062 1 1>}
  n43 = DotProduct(A=CameraVectorWS(), B=PixelNormalWS())
  n44 = ScalarParameter() {defaultValue=3, parameterName=FresnelMaskPower}
  NamedRerouteDeclaration(None=Saturate(None=Power(Base=@n43, Exp=@n44))) {name=FresnelMask, nodeColor=<rgba 0 1 0.1328 1>}
  NamedRerouteDeclaration(None=ComponentMask(None=VectorParameter() {parameterName=Block_HitLocation, defaultValue=<rgba 0.4427 0.4867 0 1>}.RGB) {g=true, r=true}) {name=Block_HitLocation, nodeColor=<rgba 0 -0.5 1 1>}
  n45 = ComponentMask(None=NamedRerouteUsage()) {r=true}
  n46 = ComponentMask(None=NamedRerouteUsage()) {r=true}
  n47 = Constant() {r=0.2}
  n48 = NamedRerouteUsage()
  n49 = Multiply(A=Time(), B=ScalarParameter() {defaultValue=2, parameterName=DebugTime})
  n50 = Multiply(
   A=Multiply(A=SphereMask(A=@n45, B=@n46, Radius=@n47) {hardnessPercent=0}, B=SphereMask(A=@n48, B=@n46, Radius=@n47) {hardnessPercent=0})
   B=NamedRerouteDeclaration(None=Frac(None=@n49)) {name=HitTimeInput, nodeColor=<rgba 1 0.2578 0 1>}
  )
  NamedRerouteDeclaration(None=@n50) {name=HitReactPitchMask, nodeColor=<rgba 0 1 0.1797 1>}
  NamedRerouteDeclaration(None=LinearInterpolate(
    A=ScalarParameter() {parameterName=HitReact_MinAdd}
    B=ScalarParameter() {defaultValue=-1, parameterName=HitReact_MaxAdd}
    Alpha=NamedRerouteUsage()
   )) {name=HitReactAddValues, nodeColor=<rgba 1 0 0.1406 1>}
  n51 = NamedRerouteUsage()
  n52 = AppendVector(A=NamedRerouteUsage(), B=NamedRerouteUsage())
  n53 = Divide(A=Ceil(None=Multiply(A=@n51, B=@n52)), B=@n52)
  n54 = ScalarParameter() {defaultValue=0.25, parameterName=Block_HitRadius}
  n55 = NamedRerouteUsage()
  NamedRerouteDeclaration(None=SphereMask(A=@n53, B=NamedRerouteUsage(), Radius=Multiply(A=@n54, B=@n55)) {hardnessPercent=50}) {name=Block_HitLocation_Mask, nodeColor=<rgba 0.0078 1 0 1>}
  n56 = ComponentMask(None=MaterialFunctionCall(Vec2=@n53) {materialFunction=&MF_Hash}.Vec2) {g=true, r=true}
  n57 = ScalarParameter() {defaultValue=0.25, parameterName=BlockHit_DistortionStrength}
  NamedRerouteDeclaration(None=Multiply(A=Multiply(A=@n56, B=@n57), B=NamedRerouteUsage())) {name=BlockHit_Distortion, nodeColor=<rgba 1 0 0.3281 1>}
  n58 = Multiply(A=NamedRerouteUsage(), B=@n33)
  NamedRerouteDeclaration(None=Divide(A=Floor(None=@n58), B=@n33)) {name=PitchGradient, nodeColor=<rgba 0 1 0.5313 1>}
  NamedRerouteDeclaration(None=StaticSwitchParameter(True=Constant() {r=2}, False=ScalarParameter() {bUseCustomPrimitiveData=true, parameterName=C_GauntletState}) {parameterName=UsePickupParamters}) {name=GauntletState, nodeColor=<rgba 1 0 0.4922 1>}
  n59 = NamedRerouteUsage()
  n60 = Time()
  n61 = LinearInterpolate(
   A=ScalarParameter() {defaultValue=1, parameterName=PitchMin_Speed}
   B=ScalarParameter() {defaultValue=2, parameterName=PitchMax_Speed}
   Alpha=NamedRerouteUsage()
  )
  n62 = ConstantBiasScale(None=Sine(None=Multiply(A=@n60, B=@n61)))
  n63 = Add(
   A=LinearInterpolate(
    A=ScalarParameter() {defaultValue=-0.5, parameterName=PitchMin_Variation}
    B=ScalarParameter() {defaultValue=0.1, parameterName=PitchMax_Variation}
    Alpha=NamedRerouteUsage()
   )
   B=Add(A=OneMinus(None=@n59), B=LinearInterpolate(Alpha=@n62) {constA=-0.5, constB=0})
  )
  NamedRerouteDeclaration(None=LinearInterpolate(
    A=ScalarParameter() {defaultValue=0.65, parameterName=PitchMin}
    B=ScalarParameter() {defaultValue=1, parameterName=PitchMax}
    Alpha=NamedRerouteDeclaration(None=@n63) {name=PitchBariationMask, nodeColor=<rgba 0.5469 1 0 1>}
   )) {name=PitchVariation, nodeColor=<rgba 1 0.2109 0 1>}
  NamedRerouteDeclaration(None=AppendVector(A=Frac(None=@n58), B=Frac(None=@n35))) {name=CellUV, nodeColor=<rgba 1 0.6094 0 1>}
  n64 = NamedRerouteUsage()
  n65 = Subtract(A=Multiply(A=@n64) {constB=2})
  n66 = MaterialFunctionCall(Float2=@n65) {materialFunction=&BreakOutFloat2Components}
  n67 = Constant()
  n68 = ScalarParameter() {defaultValue=0.95, parameterName=CellWidth}
  n69 = ScalarParameter() {defaultValue=0.9, parameterName=CellHardness}
  n70 = SphereMask(A=@n66.G, B=@n67, Radius=@n68, Hardness=@n69)
  n71 = LinearInterpolate(
   A=Multiply(A=SphereMask(A=@n66.R, B=@n67, Radius=@n68, Hardness=@n69), B=@n70)
   B=SphereMask(
    A=@n65
    B=Constant2Vector()
    Radius=ScalarParameter() {defaultValue=0.75, parameterName=CellDotScale}
    Hardness=ScalarParameter() {defaultValue=0.9, parameterName=Cell_DotHardness}
   )
   Alpha=ScalarParameter() {parameterName=CellShape1}
  )
  n72 = ScalarParameter() {parameterName=CellShape0}
  n73 = MaterialFunctionCall(UVs=Rotator(Coordinate=@n64, Time=ScalarParameter() {defaultValue=6.5, parameterName=ShapeRoatation}), Texture Scale=Constant() {r=1.25}) {materialFunction=&ScaleUVsByCenter}
  NamedRerouteDeclaration(None=LinearInterpolate(
    A=LinearInterpolate(A=@n71, B=@n70, Alpha=@n72)
    B=TextureSample(UVs=@n73.UVs) {samplerSource=SSM_Clamp_WorldGroupSettings, texture=&/T-Banners-Icons-LightningHeart-L}.R
    Alpha=ScalarParameter() {parameterName=CellShape2}
   )) {name=Stem_CellMask, nodeColor=<rgba 0.2891 0 1 1>}
  n74 = TwoSidedSign()
  NamedRerouteDeclaration(None=OneMinus(None=Saturate(None=@n74))) {name=TwoSidedMask, nodeColor=<rgba 0 0.2734 1 1>}
  n75 = NamedRerouteUsage()
  n76 = Constant2Vector() {g=0.5, r=0.5}
  n77 = ScalarParameter() {defaultValue=1, parameterName=BGCell_Radius}
  n78 = ScalarParameter() {defaultValue=0.25, parameterName=BGCell_Hardness}
  n79 = MaterialFunctionCall(Float2=NamedRerouteUsage()) {materialFunction=&BreakOutFloat2Components}
  n80 = Constant()
  n81 = ScalarParameter() {defaultValue=1, parameterName=BorderSCale}
  n82 = ScalarParameter() {defaultValue=0.5, parameterName=BorderHardness}
  n83 = Multiply(A=Multiply(A=SphereMask(A=@n79.R, B=@n80, Radius=@n81, Hardness=@n82), B=SphereMask(A=@n79.G, B=@n80, Radius=@n81, Hardness=@n82))) {constB=1.5}
  NamedRerouteDeclaration(None=Multiply(A=SphereMask(A=@n75, B=@n76, Radius=@n77, Hardness=@n78) {hardnessPercent=0}, B=Saturate(None=@n83))) {name=BorderMask, nodeColor=<rgba 1 0 0.9141 1>}
  NamedRerouteUsage()
  NamedRerouteUsage()
  OneMinus()
  OneMinus(None=@n41)
  Saturate(None=@n50)
  Transform(None=@n37.RGB) {transformSourceType=TRANSFORMSOURCE_World, transformType=TRANSFORM_Tangent}
 ]
}
