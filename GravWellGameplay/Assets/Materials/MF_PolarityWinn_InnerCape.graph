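# Material function that adds an animated emissive noise effect to the inner cape of PolarityWinn, toggled by the UseInnerCape static switch.
# Local XZ position drives two opposite-scrolling noise samples, with UVs computed per vertex through VertexInterpolators.
# Their squared difference tints InnerCapeEmissiveColor, feeding an emissive add and a material attributes blend intended for the selected FXMask channel.
# A time-based sine branch and a biased add are left unconnected to the output.
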MaterialFunction /GravWellGameplay/Assets/Materials/MF_PolarityWinn_InnerCape
graph {nodes [
  n1 = ComponentMask(None=LocalPosition().XYZ) {b=true, r=true}
  n2 = Multiply(
   A=Multiply(A=Multiply(A=@n1) {constB=0.01}, B=ScalarParameter() {defaultValue=0.15, parameterName=NoiseScale}) {constB=2}
   B=Constant2Vector() {g=0.25, r=1}
  )
  n3 = Multiply(
   A=Time()
   B=ComponentMask(None=VectorParameter() {parameterName=InnerCapeNoiseSpeed, defaultValue=<rgba 0.02 0.03 0 1>}.RGB) {g=true, r=true}
  ) {constB=0.1}
  n4 = Add(A=@n2, B=@n3)
  n5 = TextureSample(UVs=VertexInterpolator(VS=Subtract(A=@n2, B=@n3)).PS) {samplerType=SAMPLERTYPE_LinearColor, texture=&HighContrastNoise}
  n6 = Subtract(
   A=TextureSample(UVs=VertexInterpolator(VS=@n4).PS) {samplerType=SAMPLERTYPE_LinearColor, texture=&HighContrastNoise}.R
   B=@n5.B
  )
  Add(A=@n6) {constB=0.15}
  n7 = Time()
  n8 = Sine(None=Add(A=@n5.B, B=Multiply(A=@n7) {constB=0.1}))
  ConstantBiasScale(None=@n8)
  n9 = NamedRerouteUsage()
  n10 = GetMaterialAttributes(None=FunctionInput() {inputType=FunctionInput_MaterialAttributes}) {attributeGetTypes [#1]}
  n11 = SetMaterialAttributes!(MaterialAttributes=@n10.MaterialAttributes, EmissiveColor=NamedRerouteUsage(), BaseColor, Metallic, Specular, Roughness) {attributeSetTypes [
    #1
    #2
    #3
    #4
    #5
   ]}
  n12 = NamedRerouteUsage()
  FunctionOutput(None=StaticSwitchParameter(True=BlendMaterialAttributes(A=@n9, B=@n11, Alpha=@n12), False=@n9) {parameterName=UseInnerCape})
  n13 = Saturate(None=@n8)
  Multiply(A=@n13, B=@n13)
  NamedRerouteDeclaration(None=@n10.EmissiveColor) {name=InEmissive, nodeColor=<rgba 0.2188 0 1 1>}
  NamedRerouteDeclaration(None=@n10.MaterialAttributes) {name=InMaterialAttributes, nodeColor=<rgba 0.9219 0 1 1>}
  n14 = Saturate(None=Multiply(A=@n6, B=@n6))
  n15 = VectorParameter() {parameterName=InnerCapeEmissiveColor, defaultValue=<rgba 0 1.6902 3 1>}
  NamedRerouteDeclaration(None=Add(A=NamedRerouteUsage(), B=Multiply(A=@n14, B=@n15.RGB))) {name=OutEmissive, nodeColor=<rgba 0.6406 1 0 1>}
  NamedRerouteDeclaration(None=ChannelMaskParameter(None=FunctionInput() {inputName=FXMask, sortPriority=1}) {maskChannel=Green, parameterName=InnerCapeMaskChannel, defaultValue=<rgba 0 1 0 0>}.R) {name=InnerCapeMask, nodeColor=<rgba 0.8281 0 1 1>}
 ]}
