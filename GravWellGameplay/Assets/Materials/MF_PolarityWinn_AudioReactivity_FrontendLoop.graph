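# Material function that adds a looping fake audio signal to MF_AudioReactivity's amplitude and spectrum when no music plays in the Frontend.
# A time-scrolled curve atlas row feeds channel-masked fake values, replaceable through custom inputs.
# These apply when real input falls below a threshold, scaled by a part-modifier enable parameter, and merge through Max behind the UseFrontendFakeAudio static switch.
# A separate cosine loop is left unconnected.
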MaterialFunction /GravWellGameplay/Assets/Materials/MF_PolarityWinn_AudioReactivity_FrontendLoop
graph {
 properties {bExposeToLibrary=true, description=Enables a fake amplitude loop when no music is playing in the Frontend}
 nodes [
  n1 = MaterialFunctionCall() {materialFunction=/Game/Packages/Fortress_SharedMaterials/Base_Material_Functions/BaseCharacterShaderFunctions/Audio/MF_AudioReactivity}
  n2 = FunctionInput(Preview=@n1.Amplitude) {
   bUsePreviewValueAsDefault=true
   description=Defaults to Amplitude output of MF_AudioReactivity
   inputName=InAmplitude
   inputType=FunctionInput_Scalar
  }
  n3 = ScalarParameter() {defaultValue=0.05, group=Audio Reactivity, parameterName=FrontendFakeAudio_DetectionThreshold, sortPriority=200}
  n4 = If(A=@n2, B=@n3, "A > B"=Constant(), "A < B"=Constant() {r=1}) {constB=0.05}
  n5 = ScalarParameter() {desc=Set this true in the part modifier BP, group=Audio Reactivity, parameterName=EnableFrontendFakeAudioLoop, sortPriority=200}
  n6 = FunctionInput(Preview=ChannelMaskParameter(None=NamedRerouteUsage()) {group=Audio Reactivity, parameterName=AmplitudeMaskChannel}.R) {
   bUsePreviewValueAsDefault=true
   description=Use this input to supply a custom animation of the Amplitude
   inputName=CustomFakeAmplitude
   inputType=FunctionInput_Scalar
   sortPriority=10
  }
  n7 = NamedRerouteDeclaration(None=Multiply(A=Multiply(A=@n4, B=@n5), B=@n6)) {name=FrontendFakeAudio, nodeColor=<rgba 0 0.6016 1 1>}
  FunctionOutput(None=StaticSwitchParameter(True=Max(A=@n2, B=@n7), False=@n2) {defaultValue=true, group=Audio Reactivity, parameterName=UseFrontendFakeAudio, sortPriority=199}) {outputName=Amplitude}
  n8 = NamedRerouteUsage()
  n9 = ChannelMaskParameter(None=@n8) {group=Audio Reactivity, maskChannel=Green, parameterName=SpectrumLowChannel, defaultValue=<rgba 0 1 0 0>}
  n10 = ChannelMaskParameter(None=@n8) {group=Audio Reactivity, maskChannel=Blue, parameterName=SpectrumMidChannel, defaultValue=<rgba 0 0 1 0>}
  n11 = ChannelMaskParameter(None=@n8) {group=Audio Reactivity, maskChannel=Alpha, parameterName=SpectrumHighChannel, defaultValue=<rgba 0 0 0 1>}
  n12 = FunctionInput(Preview=MaterialFunctionCall(R=ChannelMaskParameter(None=@n8) {group=Audio Reactivity, parameterName=SpectrumBassChannel}.R, G=@n9.R, B=@n10.R, A=@n11.R) {materialFunction=&AppendMany}.RGBA) {
   bUsePreviewValueAsDefault=true
   description=Use this input to supply a custom animation of the Spectrum
   inputName=CustomFakeSpectrum
   inputType=FunctionInput_Vector4
   sortPriority=10
  }
  n13 = FunctionInput(Preview=@n1.Spectrum) {
   bUsePreviewValueAsDefault=true
   description=Defaults to Spectrum output of MF_AudioReactivity
   inputName=InSpectrum
   inputType=FunctionInput_Vector4
   sortPriority=1
  }
  n14 = Length(None=@n13) {is=class MaterialExpressionLength}
  n15 = Constant()
  n16 = Constant() {r=1}
  n17 = Multiply(A=@n12, B=Multiply(A=If(A=@n14, B=@n3, "A > B"=@n15, "A < B"=@n16) {constB=0.05}, B=@n5))
  FunctionOutput(None=StaticSwitchParameter(True=Max(A=@n13, B=@n17), False=@n13) {defaultValue=true, group=Audio Reactivity, parameterName=UseFrontendFakeAudio, sortPriority=199}) {outputName=Spectrum, sortPriority=10}
  n18 = Divide(A=Time(), B=ScalarParameter() {defaultValue=1.3, group=Audio Reactivity, parameterName=FrontendFakeAudio_LoopLength, sortPriority=200})
  Multiply(
   A=ConstantBiasScale(None=Cosine(None=@n18))
   B=ScalarParameter() {defaultValue=0.5, group=Audio Reactivity, parameterName=FrontendFakeAudio_Amplitude, sortPriority=200}
  )
  n19 = Time()
  n20 = ScalarParameter() {defaultValue=0.5, group=Audio Reactivity, parameterName=FakeAudioCurveSpeed}
  n21 = CurveAtlasRowParameter(CurveTime=Frac(None=Multiply(A=@n19, B=@n20))) {
   atlas=/GravWellGameplay/Assets/Textures/CA_Frontend_FakeAudioReactivity
   curve=/GravWellGameplay/Assets/Textures/C_Frontend_FakeAudioCurve_1
   group=Audio Reactivity
   parameterName=FakeAudioCurveAtlasIndex
  }
  NamedRerouteDeclaration(None=Multiply(A=AppendVector(A=@n21, B=@n21), B=ScalarParameter() {defaultValue=1, group=Audio Reactivity, parameterName=FakeAudioReactivityMaxIntensity})) {name=FakeAudioData, nodeColor=<rgba 0.1719 0 1 1>}
 ]
}
