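# Material function that layers animated, audio-reactive PolarityWinn wave lines onto incoming material attributes, with an FXMask channel and UseWaveEffect switch.
# A flow-map gradient, noise distortion and time drive sine waves shaped into glowing lines over two- or four-color gradients with flickering fresnel.
# Amplitude from a frontend-loop audio function modulates the effect, and an optional local-position mask retunes the arms.
# Results set emissive, tinted by time of day, and base color, while biasing metallic, specular and roughness.
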MaterialFunction /GravWellGameplay/Assets/Materials/MF_PolarityWinn_Waves
graph {nodes [
  n1 = NamedRerouteUsage()
  n2 = GetMaterialAttributes(None=FunctionInput() {inputType=FunctionInput_MaterialAttributes}) {attributeGetTypes [#1, #2, #3]}
  n3 = ScalarParameter() {group=PolarityWinn, parameterName=MetallicBias}
  n4 = ScalarParameter() {defaultValue=0.25, group=PolarityWinn, parameterName=SpecularBias}
  n5 = ScalarParameter() {defaultValue=-0.4, group=PolarityWinn, parameterName=RoughnessBias}
  n6 = SetMaterialAttributes!(
   MaterialAttributes=@n2.MaterialAttributes
   EmissiveColor=NamedRerouteUsage()
   BaseColor=NamedRerouteUsage()
   Metallic=Saturate(None=Add(A=@n2.Metallic, B=@n3))
   Specular=Saturate(None=Add(A=@n2.Specular, B=@n4))
   Roughness=Saturate(None=Add(A=@n2.Roughness, B=@n5))
  ) {attributeSetTypes [
    #4
    #5
    #1
    #2
    #3
   ]}
  n7 = NamedRerouteUsage()
  FunctionOutput(None=StaticSwitchParameter(True=BlendMaterialAttributes(A=@n1, B=@n6, Alpha=@n7), False=NamedRerouteUsage()) {defaultValue=true, group=PolarityWinn, parameterName=UseWaveEffect})
  NamedRerouteDeclaration(None=Multiply(
    A=MaterialFunctionCall!(InAmplitude, InSpectrum, CustomFakeSpectrum, CustomFakeAmplitude) {materialFunction=&/MF_PolarityWinn_AudioReactivity_FrontendLoop}.Amplitude
    B=ScalarParameter() {defaultValue=1, group=PolarityWinn, parameterName=EnableReactivity}
   )) {name=Amplitude, nodeColor=<rgba 1 0 0.6562 1>}
  n8 = VectorParameter() {group=PolarityWinn, parameterName=BGColor2, defaultValue=<rgba 0.7562 0 1 1>}
  n9 = VectorParameter() {group=PolarityWinn, parameterName=BGColor1, defaultValue=<rgba 0.1142 0.4555 1 1>}
  n10 = VectorParameter() {group=PolarityWinn, parameterName=BGColor3, defaultValue=<rgba 0.079 1 0.4383 1>}
  n11 = VectorParameter() {group=PolarityWinn, parameterName=BGColor4, defaultValue=<rgba 1 0.0845 0 1>}
  n12 = NamedRerouteUsage()
  n13 = NamedRerouteUsage()
  n14 = ScalarParameter() {group=PolarityWinn, parameterName=UseBGUV2}
  n15 = ScalarParameter() {defaultValue=0.5, group=PolarityWinn, parameterName=BGGradientBias}
  n16 = ScalarParameter() {defaultValue=0.1, group=PolarityWinn, parameterName=BGGradientBiasArm}
  n17 = NamedRerouteUsage()
  n18 = ScalarParameter() {defaultValue=1, group=PolarityWinn, parameterName=BGGradientScale}
  n19 = Multiply(
   A=Add(A=LinearInterpolate(A=@n12, B=@n13, Alpha=@n14), B=LinearInterpolate(A=@n15, B=@n16, Alpha=@n17))
   B=LinearInterpolate(A=@n18, B=Divide(A=@n18) {constB=1.2}, Alpha=NamedRerouteUsage())
  )
  n20 = Abs(None=Sine(None=@n19))
  n21 = ScalarParameter() {defaultValue=1, group=PolarityWinn, parameterName=BGGradientExp}
  n22 = Saturate(None=LinearInterpolate(A=@n19, B=Power(Base=@n20, Exp=@n21) {constExponent=0.5}, Alpha=ScalarParameter() {group=PolarityWinn, parameterName=AnimateBGGradient}))
  n23 = NamedRerouteUsage()
  n24 = LinearInterpolate(Alpha=OneMinus(None=Multiply(A=@n23, B=@n23))) {constA=0.25}
  n25 = NamedRerouteUsage()
  n26 = NamedRerouteUsage()
  n27 = Saturate(None=Add(A=ConstantBiasScale(None=@n25) {bias=-0.5, scale=2}, B=LinearInterpolate(Alpha=@n26) {constB=0.25}) {constB=0.1})
  n28 = LinearInterpolate(
   A=LinearInterpolate(
    A=ScalarParameter() {defaultValue=0.1, group=PolarityWinn, parameterName=BaseEmissiveAmount}
    B=ScalarParameter() {defaultValue=0.1, group=PolarityWinn, parameterName=BaseEmissiveAmountArm}
    Alpha=NamedRerouteUsage()
   )
   Alpha=Multiply(A=@n27, B=@n27)
  )
  n29 = Power(Base=NamedRerouteUsage()) {constExponent=4}
  n30 = NamedRerouteDeclaration(None=Add(A=LinearInterpolate(A=@n28, Alpha=@n29), B=NamedRerouteUsage())) {name=GrayscaleValues, nodeColor=<rgba 0 0.8828 1 1>}
  n31 = ScalarParameter() {defaultValue=1, group=PolarityWinn, parameterName=UseFresnel}
  n32 = Abs(None=NamedRerouteUsage())
  n33 = Multiply(A=@n32, B=@n32)
  n34 = Multiply(
   A=StaticSwitchParameter(
    True=MaterialFunctionCall!(A=@n8.RGB, B=@n9.RGB, C=@n10.RGB, D=@n11.RGB, Alpha=@n22) {materialFunction=&/Lerp_4Color}.Result
    False=LinearInterpolate(A=@n9.RGB, B=@n8.RGB, Alpha=@n22)
   ) {group=PolarityWinn, parameterName=Use4BGColors}
   B=Multiply(A=LinearInterpolate(A=@n24, B=@n30, Alpha=@n31) {constA=1}, B=LinearInterpolate(Alpha=@n33) {constA=0.8, constB=1.3})
  )
  n35 = NamedRerouteUsage()
  n36 = ScalarParameter() {group=PolarityWinn, parameterName=FresnelColorGradientBias}
  n37 = LinearInterpolate(
   A=VectorParameter() {group=PolarityWinn, parameterName=FresnelColor1, defaultValue=<rgba 0.0071 0.2246 1 1>}.RGB
   B=VectorParameter() {group=PolarityWinn, parameterName=FresnelColor2, defaultValue=<rgba 1 0.0509 0.8469 0.5>}.RGB
   Alpha=Saturate(None=Add(A=@n35, B=@n36) {constB=0})
  )
  n38 = Saturate(None=NamedRerouteUsage())
  n39 = ScalarParameter() {defaultValue=1, group=PolarityWinn, parameterName=SteppedFresnelIntensity}
  n40 = Saturate(None=Add(A=NamedRerouteUsage(), B=Multiply(A=@n38, B=@n39)))
  n41 = StaticSwitchParameter(True=LinearInterpolate(A=@n34, B=Multiply(A=@n37, B=@n40), Alpha=@n40), False=@n34) {defaultValue=true, group=PolarityWinn, parameterName=UseFresnel}
  n42 = VectorParameter() {group=PolarityWinn, parameterName=LinesColor1, defaultValue=<rgba 0.0071 0.2246 1 1>}
  n43 = ScalarParameter() {group=PolarityWinn, parameterName=LinesColorGradientBias}
  n44 = ScalarParameter() {defaultValue=-0.3, group=PolarityWinn, parameterName=LinesColorGradientBiasArm}
  n45 = Add(A=NamedRerouteUsage(), B=LinearInterpolate(A=@n43, B=Add(A=@n43, B=@n44) {constB=0.1}, Alpha=NamedRerouteUsage())) {constB=0}
  n46 = StaticSwitchParameter(
   True=LinearInterpolate(
    A=@n42.RGB
    B=VectorParameter() {group=PolarityWinn, parameterName=LinesColor2, defaultValue=<rgba 1 0.0509 0.8469 0.5>}.RGB
    Alpha=Saturate(None=@n45)
   )
   False=@n42.RGB
  ) {defaultValue=true, group=PolarityWinn, parameterName=LineUse2Color}
  n47 = Desaturation(None=@n46, Fraction=ScalarParameter() {group=PolarityWinn, parameterName=LinesSaturation})
  n48 = ScalarParameter() {defaultValue=0.15, group=PolarityWinn, parameterName=LinesTopBottomFadeMin}
  n49 = ConstantBiasScale(None=NamedRerouteUsage()) {bias=-0.5, scale=2}
  n50 = ScalarParameter() {group=PolarityWinn, parameterName=LinesTopBottomFadeBias}
  n51 = OneMinus(None=Abs(None=Add(A=@n49, B=@n50)))
  n52 = Multiply(A=NamedRerouteUsage(), B=Saturate(None=SmoothStep(Min=@n48, Value=@n51) {constMin=0.15}))
  n53 = OneMinus(None=Saturate(None=NamedRerouteUsage()))
  n54 = Power(Base=Abs(None=NamedRerouteUsage())) {constExponent=1.5}
  n55 = Multiply(A=@n46, B=Multiply(A=Multiply(A=@n52, B=@n53), B=LinearInterpolate(Alpha=@n54) {constA=0.5, constB=1.3}))
  n56 = NamedRerouteUsage()
  n57 = ScalarParameter() {defaultValue=4, group=PolarityWinn, parameterName=ReactivityIntensityAmplitudeExponent}
  n58 = Multiply(
   A=Desaturation(None=LinearInterpolate(A=@n41, B=@n47, Alpha=@n55), Fraction=Multiply(A=@n56) {constB=-0.5})
   B=LinearInterpolate(Alpha=Power(Base=@n56, Exp=@n57) {constExponent=4}) {constA=1, constB=1.75}
  )
  n59 = Multiply(A=@n58, B=ScalarParameter() {defaultValue=1, group=PolarityWinn, parameterName=EmissiveAmount}) {constB=0}
  n60 = Multiply(A=@n59) {constB=0.8}
  NamedRerouteDeclaration(None=Multiply(A=MaterialFunctionCall(Color_Morning=@n59, Color_Day=@n59, Color_Evening=@n60, Color_Night=@n60) {materialFunction=&Color_Time_of_Day}.Result)) {name=OutEmissive, nodeColor=<rgba 0.8516 1 0 1>}
  n61 = TextureCoordinate() {coordinateIndex=1}
  NamedRerouteDeclaration(None=ComponentMask(None=@n61) {g=true}) {name=UV2-V, nodeColor=<rgba 0.4531 1 0 1>}
  n62 = OneMinus(None=NamedRerouteUsage())
  n63 = NamedRerouteUsage()
  n64 = Abs(None=@n63)
  n65 = Constant() {r=1}
  n66 = Saturate(None=@n63)
  n67 = Power(Base=Abs(None=NamedRerouteUsage()), Exp=Constant() {r=4})
  n68 = Add(
   A=LinearInterpolate(
    A=ScalarParameter() {defaultValue=1, group=PolarityWinn, parameterName=LineThickness1}
    B=ScalarParameter() {defaultValue=0.5, group=PolarityWinn, parameterName=LineThickness2}
    Alpha=Ceil(None=@n66)
   )
   B=LinearInterpolate(B=ScalarParameter() {defaultValue=0.2, group=PolarityWinn, parameterName=LineWipeMax}, Alpha=@n67) {constA=-0.1, constB=0.2}
  )
  n69 = Multiply(A=@n67) {constB=0.05}
  n70 = ScalarParameter() {defaultValue=1, group=PolarityWinn, parameterName=LineExponentTop}
  n71 = ScalarParameter() {defaultValue=5, group=PolarityWinn, parameterName=LineExponentBottom}
  n72 = Saturate(None=Power(Base=NamedRerouteUsage()) {constExponent=1})
  n73 = Saturate(None=Power(
    Base=SphereMask(A=@n64, B=@n65, Radius=@n68, Hardness=@n69) {hardnessPercent=0}
    Exp=LinearInterpolate(A=@n70, B=@n71, Alpha=@n72) {constB=4}
   ) {constExponent=1})
  n74 = OneMinus(None=NamedRerouteUsage())
  n75 = ScalarParameter() {defaultValue=5, group=PolarityWinn, parameterName=LineCoreExponent}
  n76 = NamedRerouteUsage()
  n77 = Multiply(A=@n76, B=@n76)
  n78 = Add(
   A=Multiply(A=Power(Base=@n73) {constExponent=4}, B=Power(Base=@n74, Exp=@n75) {constExponent=5})
   B=Multiply(A=@n73, B=Multiply(A=@n77) {constB=2})
  )
  n79 = ScalarParameter() {defaultValue=1.25, group=PolarityWinn, parameterName=LinesCoreIntensity}
  n80 = Multiply(
   A=Power(Base=Saturate(None=@n62), Exp=ScalarParameter() {defaultValue=12, group=PolarityWinn, parameterName=LinesTopFade}) {constExponent=12}
   B=Add(A=@n73, B=Multiply(A=@n78, B=@n79) {constB=5})
  )
  n81 = NamedRerouteUsage()
  n82 = Multiply(A=Multiply(A=Add(A=@n80, B=@n81), B=@n80), B=ScalarParameter() {defaultValue=0.5, group=PolarityWinn, parameterName=LinesIntensity})
  n83 = SphereMask(A=Saturate(None=NamedRerouteUsage()), B=Constant() {r=1}) {attenuationRadius=1.5, hardnessPercent=0}
  NamedRerouteDeclaration(None=LinearInterpolate(A=@n82, B=Add(A=@n83, B=@n82), Alpha=ScalarParameter() {group=PolarityWinn, parameterName=UseSecondaryLine})) {name=Lines, nodeColor=<rgba 0.2891 0 1 1>}
  n84 = Multiply(A=LocalPosition() {localOrigin=InstancePreSkinning}.XYZ) {constB=0.01}
  n85 = Abs(None=ComponentMask(None=@n84) {r=true})
  n86 = ComponentMask(None=@n84) {b=true}
  n87 = Time()
  n88 = ComponentMask(None=VectorParameter() {group=PolarityWinn, parameterName=NoiseSpeed, defaultValue=<rgba -0.1 0.1 0 1>}.RGB) {g=true, r=true}
  n89 = Multiply(
   A=Add(A=AppendVector(A=@n85, B=@n86), B=Multiply(A=@n87, B=@n88) {constB=0.1})
   B=ScalarParameter() {defaultValue=1, group=PolarityWinn, parameterName=NoiseScale}
  ) {constB=2}
  NamedRerouteDeclaration(None=TextureSample(UVs=VertexInterpolator(VS=@n89).PS) {samplerType=SAMPLERTYPE_LinearColor, texture=&HighContrastNoise}.B) {name=Noise, nodeColor=<rgba 0.2891 1 0 1>}
  NamedRerouteDeclaration(None=Multiply(A=@n58, B=ScalarParameter() {group=PolarityWinn, parameterName=BaseColorAmount}) {constB=0}) {name=OutBaseColor, nodeColor=<rgba 0 0.8125 1 1>}
  n90 = OneMinus(None=SphereMask(A=LocalPosition() {localOrigin=InstancePreSkinning}.XYZ, B=Constant()) {attenuationRadius=100})
  n91 = VertexInterpolator(VS=@n90)
  NamedRerouteDeclaration(None=StaticSwitchParameter(True=ShaderStageSwitch(PixelShader=@n91.PS, VertexShader=@n90), False=Constant()) {group=PolarityWinn, parameterName=UseArmsMask}) {name=ArmsMask, nodeColor=<rgba 0 1 0.3672 1>}
  NamedRerouteDeclaration(None=@n2.MaterialAttributes) {name=InMaterialAttributes, nodeColor=<rgba 0.9219 0 1 1>}
  n92 = Multiply(A=NamedRerouteUsage()) {constB=0.5}
  n93 = Multiply(A=Time(), B=ScalarParameter() {defaultValue=0.3, group=PolarityWinn, parameterName=WipeSpeed}) {constB=0.3}
  NamedRerouteDeclaration(None=Sine(None=Add(A=@n92, B=@n93))) {name=Wipe, nodeColor=<rgba 0.8281 0 1 1>}
  NamedRerouteDeclaration(None=ChannelMaskParameter(None=FunctionInput() {inputName=FXMask, sortPriority=1}) {group=PolarityWinn, parameterName=WavesMask}.R) {name=WavesMask, nodeColor=<rgba 0.8281 0 1 1>}
  n94 = ScalarParameter() {defaultValue=70, group=PolarityWinn, parameterName=WaveformNoiseScale}
  n95 = LinearInterpolate(A=@n94, B=Multiply(A=@n94) {constB=0.75}, Alpha=NamedRerouteUsage())
  n96 = LocalPosition() {localOrigin=InstancePreSkinning}
  n97 = Constant3Vector() {constant=<rgba 1 1 0.5 0>}
  n98 = Multiply(A=@n96.XYZ, B=@n97)
  n99 = VectorParameter() {group=PolarityWinn, parameterName=WaveformSpeed, defaultValue=<rgba 1 1 -40 1>}
  n100 = ScalarParameter() {defaultValue=0.7, group=PolarityWinn, parameterName=WaveFormSpeedArmMult}
  n101 = Multiply(A=Time(), B=LinearInterpolate(A=@n99.RGB, B=Multiply(A=@n99.RGB, B=@n100), Alpha=NamedRerouteUsage())) {constB=0.1}
  n102 = MaterialFunctionCall(
   TextureObject=TextureObject() {samplerType=SAMPLERTYPE_LinearColor, texture=&HighContrastNoise}
   TextureSize=@n95
   WorldPosition=VertexInterpolator(VS=Add(A=@n98, B=@n101)).PS
  ) {materialFunction=&WorldAlignedTexture}
  n103 = MaterialFunctionCall(Color=Multiply(A=Multiply(A=@n96.XYZ, B=@n97)) {constB=0.01}) {materialFunction=&SplitComponents}
  n104 = Add(
   A=MaterialFunctionCall(R=Abs(None=@n103.R), G=@n103.G, B=@n103.G) {materialFunction=&AppendMany}.RGB
   B=@n101
  )
  n105 = VertexInterpolator(VS=ComponentMask(None=Multiply(A=@n104, B=@n95)) {b=true, r=true})
  NamedRerouteDeclaration(None=StaticSwitchParameter(
    True=MaterialFunctionCall(Color=@n102.XY Texture) {materialFunction=&SplitComponents}.G
    False=TextureSample(UVs=@n105.PS) {samplerType=SAMPLERTYPE_LinearColor, texture=&HighContrastNoise}.G
   ) {group=PolarityWinn, parameterName=UseWaveformTriplanar}) {name=WaveFormNoise, nodeColor=<rgba 0.9219 1 0 1>}
  n106 = Time()
  n107 = ScalarParameter() {defaultValue=3, group=PolarityWinn, parameterName=SecondaryLinesSpeed}
  n108 = ScalarParameter() {defaultValue=-0.1, group=PolarityWinn, parameterName=LinesDistBias}
  n109 = LinearInterpolate(A=@n108, B=Add(A=@n108, B=ScalarParameter() {group=PolarityWinn, parameterName=LinesDistBiasArm}), Alpha=NamedRerouteUsage())
  n110 = LinearInterpolate(A=TextureCoordinate(), B=@n61, Alpha=ScalarParameter() {group=PolarityWinn, parameterName=FlowMapUseUV2})
  n111 = Subtract(A=NamedRerouteUsage()) {constB=0.5}
  n112 = ScalarParameter() {defaultValue=0.05, group=PolarityWinn, parameterName=WaveFormNoiseIntensity}
  n113 = Multiply(
   A=Multiply(A=Multiply(A=@n111) {constB=2}, B=Constant2Vector() {g=1})
   B=LinearInterpolate(A=@n112, B=Multiply(A=@n112) {constB=0.5}, Alpha=NamedRerouteUsage())
  ) {constB=0.1}
  n114 = Power(Base=NamedRerouteUsage(), Exp=ScalarParameter() {defaultValue=2, group=PolarityWinn, parameterName=WaveformAmplitudeExponent}) {constExponent=4}
  n115 = TextureSampleParameter2D(UVs=Add(A=@n110, B=Multiply(A=@n113, B=@n114))) {
   desc=Undeformed Gradient
   group=PolarityWinn
   parameterName=TextureFlowMap
   samplerType=SAMPLERTYPE_LinearGrayscale
   texture=/GravWellGameplay/Assets/Textures/T_PolarityWinn_Field_Character_Flowmap
  }
  n116 = LinearInterpolate(A=@n115.R, B=OneMinus(None=@n115.R), Alpha=ScalarParameter() {defaultValue=1, group=PolarityWinn, parameterName=LinesDistInvert})
  n117 = Multiply(
   A=Power(Base=Add(A=@n109, B=@n116), Exp=ScalarParameter() {defaultValue=1.5, group=PolarityWinn, parameterName=LinesDistPow})
   B=ScalarParameter() {defaultValue=-1, group=PolarityWinn, parameterName=LinesDistInt}
  ) {constB=0.2}
  n118 = Add(A=Add(A=Multiply(A=@n106, B=@n107) {constB=0.2}) {constB=0.2}, B=@n117)
  n119 = ScalarParameter() {defaultValue=0.5, group=PolarityWinn, parameterName=SecondaryLinesAmount}
  NamedRerouteDeclaration(None=Sine(None=Multiply(A=@n118, B=@n119) {constB=7})) {name=SecondarySineWaves, nodeColor=<rgba 0 1 0.8594 1>}
  n120 = TextureSampleParameter2D(UVs=@n110) {
   desc=Undeformed Gradient
   group=PolarityWinn
   parameterName=TextureFlowMap
   samplerType=SAMPLERTYPE_LinearGrayscale
   texture=&T_PolarityWinn_Field_Character_Flowmap
  }
  NamedRerouteDeclaration(None=@n120.R) {name=VerticalGradient, nodeColor=<rgba 0 1 0.9063 1>}
  n121 = LinearInterpolate(
   A=NamedRerouteUsage()
   B=NamedRerouteUsage()
   Alpha=ScalarParameter() {defaultValue=1, group=PolarityWinn, parameterName=BottomGradientUseUV2}
  )
  n122 = MaterialFunctionCall!(Time, Speed Scalar=Constant() {r=1}) {materialFunction=&/MF_SimpleFlicker}
  n123 = Saturate(None=Add(A=ConstantBiasScale(None=@n121) {bias=-0.5, scale=2}, B=Multiply(A=@n122.Result) {constB=0.075}))
  n124 = LinearInterpolate(B=ScalarParameter() {defaultValue=8, group=PolarityWinn, parameterName=BottomGradientArmAdjust}, Alpha=NamedRerouteUsage()) {constA=6, constB=0.5}
  NamedRerouteDeclaration(None=Multiply(
    A=Power(Base=@n123, Exp=@n124) {constExponent=4}
    B=ScalarParameter() {defaultValue=10, group=PolarityWinn, parameterName=BottomGradientIntensity}
   ) {constB=10}) {name=BottomGradient, nodeColor=<rgba 0 0.7891 1 1>}
  n125 = Subtract(A=NamedRerouteUsage(), B=ScalarParameter() {defaultValue=0.25, group=PolarityWinn, parameterName=FresnelBias}) {constB=0.25}
  n126 = MaterialFunctionCall(Speed Scalar=Constant() {r=2}) {materialFunction=&MF_SimpleFlicker}
  n127 = NamedRerouteDeclaration(None=@n126.Result) {name=Flicker, nodeColor=<rgba 1 0 0.2344 1>}
  n128 = NamedRerouteDeclaration(None=Fresnel(ExponentIn=LinearInterpolate(Alpha=@n127) {constA=0.8}) {baseReflectFraction=0, exponent=1}) {name=Fresnel, nodeColor=<rgba 0.9922 0 1 1>}
  n129 = Multiply(A=Multiply(A=@n128, B=@n128))
  NamedRerouteDeclaration(None=Saturate(None=Add(A=@n125, B=@n129))) {name=SteppedFresnel, nodeColor=<rgba 1 0 0.6094 1>}
  n130 = Transform(None=Constant3Vector() {constant=<rgba 0 1 0 0>})
  n131 = Add(A=LinearInterpolate(Alpha=@n126.Result) {constA=0.5, constB=2.5}, B=Multiply(A=NamedRerouteUsage()) {constB=5})
  n132 = PixelNormalWS()
  NamedRerouteDeclaration(None=OneMinus(None=MaterialFunctionCall(Axis=@n130, Exponent=@n131, Normal=@n132) {materialFunction=&AxisAlignedFresnel}.Fresnel)) {name=AxisFresnel, nodeColor=<rgba 0.8516 1 0 1>}
  n133 = Multiply(A=Time(), B=ScalarParameter() {defaultValue=0.2, group=PolarityWinn, parameterName=LinesSpeed}) {constB=0.2}
  n134 = ScalarParameter() {defaultValue=7, group=PolarityWinn, parameterName=LinesAmount}
  n135 = Multiply(
   A=Add(A=@n117, B=NamedRerouteDeclaration(None=@n133) {name=Time, nodeColor=<rgba 1 0.0469 0 1>})
   B=LinearInterpolate(A=@n134, B=Multiply(A=@n134), Alpha=NamedRerouteUsage())
  ) {constB=7}
  n136 = ComponentMask(None=Add(A=TextureCoordinate() {coordinateIndex=1}) {constB=0.2}) {g=true}
  n137 = Saturate(None=Power(Base=ConstantBiasScale(None=@n136) {bias=-0.5, scale=2}) {constExponent=1})
  n138 = Abs(None=ComponentMask(None=PreSkinnedNormal()) {g=true})
  n139 = VertexInterpolator(VS=Saturate(None=Power(Base=@n138)))
  n140 = Add(A=@n135, B=Multiply(A=Multiply(A=@n137, B=@n139.PS)) {constB=-0.25})
  n141 = ScalarParameter() {defaultValue=1, group=PolarityWinn, parameterName=UseLineExtraDeform}
  NamedRerouteDeclaration(None=Sine(None=LinearInterpolate(A=@n135, B=@n140, Alpha=@n141))) {name=SineWaves, nodeColor=<rgba 0.8281 0 1 1>}
  n142 = ScalarParameter() {defaultValue=75, group=PolarityWinn, parameterName=TopGradientExponent}
  n143 = Divide(A=@n142, B=LinearInterpolate(Alpha=NamedRerouteUsage()) {constA=2, constB=1.5})
  n144 = NamedRerouteUsage()
  n145 = Multiply(A=Abs(None=@n144), B=LinearInterpolate(Alpha=Ceil(None=@n144)) {constA=0.5})
  n146 = Power(Base=OneMinus(None=NamedRerouteUsage()), Exp=Subtract(A=@n142, B=Multiply(A=@n143, B=@n145) {constB=50})) {constExponent=64}
  NamedRerouteDeclaration(None=Multiply(A=Saturate(None=@n146), B=Constant() {r=1}) {constB=4}) {name=UpperDetail, nodeColor=<rgba 1 0.1875 0 1>}
 ]}
