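# Parent material for the PolarityWinn Gauntlet character, layering character-specific effects between the shared 2023 CharacterParent PreFX and PostFX functions.
# PreFX output passes through a Waves function, then an InnerCape function masked by the Tech Art Mask, then a GauntletEquip function driven by pre-skinned local position.
# It uses Material Attributes with expression-selected shading models and supports skeletal, cloth, Nanite, instanced and Niagara meshes.
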Material /GravWellGameplay/Assets/Materials/M_PolarityWinn_Gauntlet_Parent
graph {
 properties {
  bHasPixelAnimation=true
  bUseMaterialAttributes=true
  bUsedWithClothing=true
  bUsedWithInstancedStaticMeshes=true
  bUsedWithMeshDeformer=true
  bUsedWithMorphTargets=true
  bUsedWithNanite=true
  bUsedWithNiagaraMeshParticles=true
  bUsedWithSkeletalMesh=true
  pixelDepthOffsetMode=PDOM_Legacy
  shadingModel=MSM_FromMaterialExpression
  usedShadingModels=Default Lit | Subsurface | Clear Coat | Subsurface Profile | Cloth
 }
 nodes [
  n1 = MaterialFunctionCall() {materialFunction=&MF_CharacterParent_2023_PreFX}
  n2 = MaterialFunctionCall!(In=@n1.Result, FXMask=@n1.Result) {materialFunction=&/MF_PolarityWinn_Waves}
  MaterialFunctionCall(
   Material=MaterialFunctionCall!(
    Material=MaterialFunctionCall!(In=@n2.Result, FXMask=@n1.Tech Art Mask) {materialFunction=&/MF_PolarityWinn_InnerCape}.Result
    PreskinnedLocalPosition=@n1.Pre skinned Local Position
   ) {materialFunction=&/MF_GauntletEquip}.Result
   M=@n1.M
   Pre Skinned Local Position=@n1.Pre skinned Local Position
   Distance Blend=@n1.M
   Fresnel=@n1.M
  ) {materialFunction=&MF_CharacterParent_2023_PostFX}
 ]
}
