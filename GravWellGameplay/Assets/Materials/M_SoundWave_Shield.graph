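# Translucent, two-sided, unlit material for a pulsing sound-wave energy shield with refraction.
# Radial gradient masks, scaled by custom primitive data `B_ActivateShield`, feed time-driven sine pulses, opacity and an activation emissive pulse.
# Panning polar-coordinate noise supplies UV and intersection distortion, alongside curve-atlas color, fresnel, depth-based edge masks and normal-based WPO.
# Named reroutes link the stages; some hit-wave nodes, including its location parameter, are left unconnected.
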Material /GravWellGameplay/Assets/Materials/M_SoundWave_Shield
graph {
 properties {
  bAutomaticallySetUsageInEditor=false
  blendMode=BLEND_Translucent
  refractionMethod=RM_IndexOfRefraction
  shadingModel=MSM_Unlit
  twoSided=true
  usedShadingModels=Unlit
 }
 nodes [
  n1 = NamedRerouteUsage()
  n2 = NamedRerouteUsage()
  n3 = Add(A=@n1, B=@n2)
  n4 = Constant2Vector() {g=0.5, r=0.5}
  n5 = Saturate(None=OneMinus(None=ScalarParameter() {bUseCustomPrimitiveData=true, parameterName=B_ActivateShield}))
  n6 = Multiply(A=ScalarParameter() {defaultValue=0.5, parameterName=BaseGradientRadius}, B=@n5)
  n7 = ScalarParameter() {parameterName=BaseGradientHardness}
  n8 = Multiply(A=Time(), B=ScalarParameter() {defaultValue=3, parameterName=PulseSpeed})
  n9 = Add(A=NamedRerouteDeclaration(None=SphereMask(A=@n3, B=@n4, Radius=@n6, Hardness=@n7)) {name=BaseGradient, nodeColor=<rgba 0 0.1562 1 1>}, B=@n8)
  Frac(None=@n9)
  Frac(None=Multiply(A=Time()))
  n10 = MaterialFunctionCall!(UV Coord=NamedRerouteUsage(), LenghtScale, RadialScale) {materialFunction=/Game/Global/MaterialFunctions/TechArt/MF_PolarCoordinates}
  n11 = Multiply(A=@n10.Result, B=ComponentMask(None=VectorParameter() {parameterName=SoftNoise_Tiling_01, defaultValue=<rgba 1 6 0 1>}.RGB) {g=true, r=true})
  n12 = Multiply(
   A=Time()
   B=ComponentMask(None=VectorParameter() {parameterName=SoftNoise_PanSpeedAndDirection_01, defaultValue=<rgba -1 0 0 1>}.RGB) {g=true, r=true}
  )
  n13 = ComponentMask(None=VectorParameter() {parameterName=SoftNoiseDistortion_Tiling_01, defaultValue=<rgba 1 6 0 1>}.RGB) {g=true, r=true}
  n14 = Time()
  n15 = ComponentMask(None=VectorParameter() {parameterName=SoftNoiseDistortion_PanSpeedAndDirection_01, defaultValue=<rgba -0.5 0 0 1>}.RGB) {g=true, r=true}
  n16 = TextureSample(UVs=Add(A=Multiply(A=@n10.Result, B=@n13), B=Multiply(A=@n14, B=@n15))) {texture=/Game/Effects/Fort_Effects/Emotes/UltraLight/Texture/T_SoftNoise_BC7}
  n17 = Multiply(A=Subtract(A=Multiply(A=@n16.RGB) {constB=2}), B=ScalarParameter() {defaultValue=0.05, parameterName=NoiseDistortion_Str})
  n18 = TextureSample(UVs=Add(A=Add(A=@n11, B=@n12), B=ComponentMask(None=@n17) {g=true, r=true})) {texture=&T_SoftNoise_BC7}
  n19 = Subtract(A=OneMinus(None=NamedRerouteUsage()), B=@n18.R)
  n20 = NamedRerouteUsage()
  LinearInterpolate(
   A=ScalarParameter() {defaultValue=1, parameterName=MinRefraction}
   B=ScalarParameter() {defaultValue=0.85, parameterName=MaxRefraction}
   Alpha=Saturate(None=Multiply(A=@n19, B=@n20))
  ) {constA=1, constB=0.95}
  n21 = NamedRerouteUsage()
  n22 = CurveAtlasRowParameter(CurveTime=@n21) {
   atlas=&/CA_PulsarShield
   curve=&/C_PulsarShield_01
   parameterName=SpectrumColorCurve
  }
  n23 = ScalarParameter() {defaultValue=2, parameterName=ColorSpectrumMult}
  n24 = ScalarParameter() {defaultValue=5, parameterName=Activation_Emissive}
  n25 = NamedRerouteUsage()
  n26 = Multiply(A=Multiply(A=@n22, B=LinearInterpolate(A=@n23, B=@n24, Alpha=@n25)), B=@n21)
  n27 = Max(A=NamedRerouteUsage(), B=Saturate(None=NamedRerouteUsage()))
  n28 = NamedRerouteUsage()
  LinearInterpolate(A=Multiply(A=@n26), B=Multiply(A=@n26) {constB=1.25}, Alpha=Max(A=Multiply(A=@n27, B=@n28), B=NamedRerouteUsage()))
  MaterialFunctionCall() {materialFunction=&ScaleUVsByCenter}
  n29 = NamedRerouteUsage()
  Max(A=NamedRerouteUsage(), B=@n29)
  n30 = NamedRerouteUsage()
  n31 = NamedRerouteUsage()
  n32 = NamedRerouteUsage()
  Max(A=NamedRerouteUsage(), B=Max(A=Multiply(A=@n30, B=@n31), B=Saturate(None=@n32)))
  n33 = NamedRerouteUsage()
  n34 = MaterialFunctionCall(Float2=Subtract(A=Multiply(A=@n33) {constB=2})) {materialFunction=&BreakOutFloat2Components}
  n35 = Constant()
  n36 = ScalarParameter() {defaultValue=1, parameterName=BorderSCale}
  n37 = ScalarParameter() {defaultValue=0.5, parameterName=BorderHardness}
  n38 = Multiply(A=SphereMask(A=@n34.R, B=@n35, Radius=@n36, Hardness=@n37), B=SphereMask(A=@n34.G, B=@n35, Radius=@n36, Hardness=@n37))
  Multiply(
   A=SphereMask(
    A=NamedRerouteUsage()
    B=Constant2Vector() {g=0.5, r=0.5}
    Radius=ScalarParameter() {defaultValue=1, parameterName=BGCell_Radius}
    Hardness=ScalarParameter() {defaultValue=0.25, parameterName=BGCell_Hardness}
   ) {hardnessPercent=0}
   B=Saturate(None=Multiply(A=@n38) {constB=1.5})
  )
  n39 = NamedRerouteUsage()
  n40 = ScalarParameter() {defaultValue=1, parameterName=OpacityMultiplier}
  n41 = LinearInterpolate(
   A=ScalarParameter() {defaultValue=0.45, parameterName=Base_Opacity_Min}
   B=ScalarParameter() {defaultValue=1, parameterName=Base_Opacity_Max}
   Alpha=Saturate(None=Multiply(A=@n39, B=@n40))
  )
  Multiply(A=@n41)
  NamedRerouteDeclaration(None=TextureCoordinate()) {name=BaseCoord, nodeColor=<rgba 1 0.2344 0 1>}
  n42 = Subtract(A=Multiply(A=NamedRerouteUsage()) {constB=2})
  n43 = ScalarParameter() {defaultValue=2, parameterName=WPO_Amt}
  NamedRerouteDeclaration(None=Multiply(A=VertexNormalWS(), B=Multiply(A=@n42, B=@n43))) {name=WPO, nodeColor=<rgba 1 0 0.6094 1>}
  n44 = MaterialFunctionCall!(
   Distance Bias=ScalarParameter() {defaultValue=50, group=IntersectingEdge, parameterName=IntersectingEdgeBias}
   Distance Divisor=ScalarParameter() {defaultValue=512, group=IntersectingEdge, parameterName=IntersectingEdgeDivisor}
   Scene Depth
   Pixel Depth
  ) {materialFunction=/Game/Effects/Fort_Effects/Materials/_Functions/MF_DistanceBehindTranslucency_01}
  NamedRerouteDeclaration(None=Power(Base=@n44.Normalized Inverted, Exp=ScalarParameter() {defaultValue=12, group=IntersectingEdge, parameterName=Intersection Edge Exponent})) {name=IntersectionEdgeMask, nodeColor=<rgba 0.1719 0 1 1>}
  n45 = Add(A=@n5) {constB=-0.25}
  NamedRerouteDeclaration(None=ConstantBiasScale(None=Sine(None=@n45))) {name=ActivationEmissivePulse, nodeColor=<rgba 1 0 0.3984 1>}
  n46 = ComponentMask(None=VectorParameter() {parameterName=SoftNoise_Tiling_02, defaultValue=<rgba 6 1 0 1>}.RGB) {g=true, r=true}
  n47 = MaterialFunctionCall(UV Coord=NamedRerouteUsage()) {materialFunction=&MF_PolarCoordinates}
  n48 = Time()
  n49 = ComponentMask(None=VectorParameter() {parameterName=SoftNoise_PanSpeedAndDirection_01, defaultValue=<rgba -1 0 0 1>}.RGB) {g=true, r=true}
  n50 = VectorParameter() {parameterName=SoftNoiseDistortion_Tiling_02, defaultValue=<rgba 6 1 0 1>}
  n51 = VectorParameter() {parameterName=SoftNoiseDistortion_PanSpeedAndDirection_01, defaultValue=<rgba -0.5 0 0 1>}
  n52 = Add(
   A=Multiply(A=@n47.Result, B=ComponentMask(None=@n50.RGB) {g=true, r=true})
   B=Multiply(A=Time(), B=ComponentMask(None=@n51.RGB) {g=true, r=true})
  )
  n53 = Subtract(A=Multiply(A=TextureSample(UVs=@n52) {texture=&T_SoftNoise_BC7}.RGB) {constB=2})
  n54 = ScalarParameter() {defaultValue=0.05, parameterName=NoiseDistortion_Str}
  n55 = Add(A=Add(A=Multiply(A=@n47.Result, B=@n46), B=Multiply(A=@n48, B=@n49)), B=ComponentMask(None=Multiply(A=@n53, B=@n54)) {g=true, r=true})
  n56 = Subtract(A=Multiply(A=TextureSample(UVs=@n55) {texture=&T_SoftNoise_BC7}.RGB) {constB=2})
  n57 = ScalarParameter() {defaultValue=0.05, parameterName=BaseUV_Distortion_Str}
  NamedRerouteDeclaration(None=ComponentMask(None=Multiply(A=@n56, B=@n57)) {g=true, r=true}) {name=BaseUV_Distortion_2, nodeColor=<rgba 0.5703 1 0 1>}
  n58 = Subtract(A=@n44.Normalized Inverted, B=NamedRerouteUsage())
  NamedRerouteDeclaration(None=Power(Base=Saturate(None=@n58)) {constExponent=7}) {name=IntersectionEdge_Test, nodeColor=<rgba 1 0 0.0703 1>}
  NamedRerouteDeclaration(None=Multiply(A=@n18.G, B=ScalarParameter() {defaultValue=0.1, parameterName=IntersectionDistortion_Str})) {name=IntersectionDistortion, nodeColor=<rgba 1 0 0.8906 1>}
  n59 = SphereMask(
   A=WorldPosition().XYZ
   B=Constant3Vector()
   Radius=ScalarParameter() {defaultValue=64, parameterName=HitRadius}
   Hardness=ScalarParameter() {parameterName=PulseRadiusHardness}
  )
  n60 = Multiply(A=Time())
  n61 = Multiply(A=Constant() {r=1}, B=OneMinus(None=Frac(None=@n60)))
  n62 = ScalarParameter() {defaultValue=0.2, parameterName=HitPulseRadius}
  n63 = ScalarParameter() {parameterName=HitPulseHardness}
  NamedRerouteDeclaration(None=Multiply(A=@n59, B=SphereMask(A=@n59, B=@n61, Radius=@n62, Hardness=@n63))) {name=HitMask, nodeColor=<rgba 0 1 0.6719 1>}
  n64 = Saturate(None=Multiply(A=NamedRerouteUsage()) {constB=1.5})
  n65 = SphereMask(A=Add(A=@n1, B=Multiply(A=@n2) {constB=0.25}), B=@n4, Radius=@n6, Hardness=@n7)
  n66 = Add(
   A=Multiply(
    A=NamedRerouteDeclaration(None=@n65) {name=BorderMask, nodeColor=<rgba 1 0 0.9141 1>}
    B=ScalarParameter() {defaultValue=3, parameterName=PulseMultiplier}
   )
   B=@n8
  )
  n67 = Sine(None=@n66)
  n68 = Saturate(None=@n67)
  NamedRerouteDeclaration(None=Multiply(A=@n64, B=NamedRerouteDeclaration(None=@n68) {name=ColorSPectrum, nodeColor=<rgba 0.7109 0 1 1>})) {name=Pulse_Alpha, nodeColor=<rgba 0 1 0.1562 1>}
  n69 = Multiply(A=@n29) {constB=2}
  n70 = LinearInterpolate(B=LinearInterpolate(A=@n41, Alpha=Saturate(None=@n69)), Alpha=NamedRerouteUsage())
  n71 = NamedRerouteUsage()
  n72 = ScalarParameter() {defaultValue=1, parameterName=BaseGradientOpacityMult}
  n73 = Multiply(A=NamedRerouteUsage(), B=ScalarParameter() {defaultValue=3, parameterName=BorderMask_Mult}) {constB=5}
  n74 = Saturate(None=Multiply(A=Multiply(A=@n71, B=@n72) {constB=2}, B=Saturate(None=@n73)))
  NamedRerouteDeclaration(None=DepthFade(Opacity=LinearInterpolate(B=@n70, Alpha=@n74)) {fadeDistanceDefault=32}) {name=Opacity_Output, nodeColor=<rgba 1 0 0.1641 1>}
  n75 = Subtract(A=Multiply(A=@n18.RGB) {constB=2})
  n76 = ScalarParameter() {defaultValue=0.05, parameterName=BaseUV_Distortion_Str}
  NamedRerouteDeclaration(None=ComponentMask(None=Multiply(A=@n75, B=@n76)) {g=true, r=true}) {name=BaseUV_Distortion, nodeColor=<rgba 0.5703 1 0 1>}
  NamedRerouteDeclaration(None=SphereMask(
    A=NamedRerouteUsage()
    B=ScalarParameter() {defaultValue=0.1, parameterName=ColorSpectrumSamplePos}
    Radius=ScalarParameter() {defaultValue=0.15, parameterName=ColorSpectrumRadius}
   ) {hardnessPercent=0}) {name=ColorSpectrumGradient, nodeColor=<rgba 1 0 0.8906 1>}
  NamedRerouteDeclaration(None=Multiply(A=ConstantBiasScale(None=@n67), B=@n64)) {name=PulseRefractionAlpha, nodeColor=<rgba 0 1 0.8594 1>}
  n77 = DotProduct(A=CameraVectorWS(), B=PixelNormalWS())
  n78 = ScalarParameter() {defaultValue=2, parameterName=FresnelStr}
  n79 = Saturate(None=Power(Base=Multiply(A=@n77, B=@n78), Exp=ScalarParameter() {defaultValue=2, parameterName=FresnelPow}))
  NamedRerouteDeclaration(None=OneMinus(None=NamedRerouteDeclaration(None=@n79) {name=FresnelMask, nodeColor=<rgba 1 0.9141 0 1>})) {name=CameraAngleMask, nodeColor=<rgba 0.4766 0 1 1>}
  NamedRerouteUsage()
  NamedRerouteUsage()
  NamedRerouteUsage()
  n80 = MaterialFunctionCall(Float3=MaterialFunctionCall(Vec3=ObjectPositionWS()) {materialFunction=&MF_Hash}.Vec3) {materialFunction=&BreakOutFloat3Components}
  Saturate(None=Add(A=Add(A=@n80.R, B=@n80.R), B=@n80.B))
  n81 = Time()
  n82 = ScalarParameter() {defaultValue=1, desc=Time-stamp, group=HitWave, parameterName=GameTime}
  Saturate(None=Divide(A=Subtract(A=@n81, B=@n82), B=ScalarParameter() {defaultValue=0.5, group=HitWave, parameterName=HitWave-DurationNew}))
  Saturate(None=Power(Base=@n41))
  n83 = ScalarParameter() {defaultValue=2, parameterName=BaseGradientRefractionMultiplier}
  Saturate(None=Power(Base=Multiply(A=@n20, B=@n83) {constB=2}))
  Saturate(None=Sine(None=@n9))
  n84 = SphereMask(A=OneMinus(None=Frac(None=@n66)), B=Constant() {r=1}) {attenuationRadius=0.5, hardnessPercent=0}
  Saturate(None=Power(Base=Multiply(A=@n84) {constB=2}))
  VectorParameter() {parameterName=BaseColor_01, defaultValue=<rgba 0.511 0.5711 0.6441 1>}
  VectorParameter() {group=HitWave, parameterName=HitWave-Loc, defaultValue=<rgba 100 -1400 -150 0>}
 ]
}
