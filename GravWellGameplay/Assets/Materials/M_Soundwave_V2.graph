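# Translucent, two-sided lit material for a soundwave effect with pulsing concentric rings, apparently related to a pulsar shield.
# Panning cloud noise in polar coordinates, warped by a second sample, supplies distortion.
# Time-scrolled radial gradients form ring masks and sine-based WPO pulses, alongside fresnel, radial and depth-intersection masks.
# Masks shape curve-atlas color, depth-faded opacity and refraction; custom primitive data drives an activation sphere mask.
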Material /GravWellGameplay/Assets/Materials/M_Soundwave_V2
graph {
 properties {
  bAutomaticallySetUsageInEditor=false
  blendMode=BLEND_Translucent
  refractionMethod=RM_IndexOfRefraction
  twoSided=true
  usedShadingModels=Default Lit
 }
 nodes [
  Frac(None=Multiply(A=Time()))
  n1 = CurveAtlasRowParameter(CurveTime=NamedRerouteUsage()) {
   atlas=&/CA_PulsarShield
   curve=&/C_PulsarShield_01
   parameterName=SpectrumColorCurve
  }
  n2 = NamedRerouteUsage()
  n3 = LinearInterpolate(
   A=VectorParameter() {parameterName=BaseColor_01, defaultValue=<rgba 0.2329 0.4626 0.7413 1>}.RGB
   B=@n1
   Alpha=Saturate(None=Multiply(A=@n2) {constB=2})
  )
  LinearInterpolate(A=Multiply(A=@n3, B=ScalarParameter() {defaultValue=1.5, parameterName=FresnelBrightness}), B=@n3, Alpha=NamedRerouteUsage())
  n4 = Multiply(A=Time())
  Multiply(A=SphereMask(
    A=NamedRerouteUsage()
    B=Constant()
    Radius=Frac(None=@n4)
    Hardness=ScalarParameter() {defaultValue=1, parameterName=ShieldActivateHardness}
   ) {attenuationRadius=0.5, hardnessPercent=1})
  NamedRerouteDeclaration(None=TextureCoordinate()) {name=BaseCoordinates, nodeColor=<rgba 0 1 0.2734 1>}
  n5 = NamedRerouteUsage()
  n6 = ScalarParameter() {defaultValue=12, parameterName=WPO_Amt}
  NamedRerouteDeclaration(None=Multiply(A=VertexNormalWS(), B=Multiply(A=@n5, B=@n6))) {name=WPO, nodeColor=<rgba 1 0 0.6094 1>}
  n7 = Multiply(A=NamedRerouteUsage(), B=NamedRerouteUsage())
  n8 = Multiply(A=ConstantBiasScale(None=NamedRerouteUsage()), B=Saturate(None=Multiply(A=@n7) {constB=2}))
  n9 = NamedRerouteUsage()
  NamedRerouteDeclaration(None=LinearInterpolate(
    A=ScalarParameter() {defaultValue=1, parameterName=Refraction_Min}
    B=ScalarParameter() {defaultValue=1.05, parameterName=Refraction_Max}
    Alpha=Multiply(A=@n8, B=@n9)
   )) {name=Refractiom_Result, nodeColor=<rgba 1 0.1641 0 1>}
  n10 = ScalarParameter() {defaultValue=0.2, parameterName=Base_BG_Opacity}
  n11 = ScalarParameter() {defaultValue=1, parameterName=PulseRingOpacity}
  n12 = NamedRerouteUsage()
  n13 = NamedRerouteUsage()
  n14 = NamedRerouteUsage()
  n15 = LinearInterpolate(
   A=LinearInterpolate(B=LinearInterpolate(A=@n10, B=@n11, Alpha=@n12), Alpha=Multiply(A=@n13, B=@n14))
   B=ScalarParameter() {defaultValue=1, parameterName=IntersectionOpacity}
   Alpha=Multiply(A=NamedRerouteUsage(), B=NamedRerouteUsage())
  )
  n16 = NamedRerouteUsage()
  NamedRerouteDeclaration(None=DepthFade(Opacity=Multiply(A=@n15, B=@n16), FadeDistance=ScalarParameter() {defaultValue=32, parameterName=DepthFade_Distance})) {name=Opacity_Result, nodeColor=<rgba 0 0.1094 1 1>}
  n17 = NamedRerouteDeclaration(None=MaterialFunctionCall!(UV Coord=NamedRerouteUsage(), LenghtScale, RadialScale) {materialFunction=/Game/Global/MaterialFunctions/TechArt/MF_PolarCoordinates}.Result) {name=PolarCoordinates, nodeColor=<rgba 1 0 0.0469 1>}
  n18 = Multiply(A=@n17, B=ComponentMask(None=VectorParameter() {parameterName=SoftNoise_Tiling_01, defaultValue=<rgba 1 6 0 1>}.RGB) {g=true, r=true})
  n19 = Multiply(
   A=Time()
   B=ComponentMask(None=VectorParameter() {parameterName=SoftNoise_PanSpeedAndDirection_01, defaultValue=<rgba -1 0 0 1>}.RGB) {g=true, r=true}
  )
  n20 = ComponentMask(None=VectorParameter() {parameterName=SoftNoiseDistortion_Tiling_01, defaultValue=<rgba 1 6 0 1>}.RGB) {g=true, r=true}
  n21 = Time()
  n22 = ComponentMask(None=VectorParameter() {parameterName=SoftNoiseDistortion_PanSpeedAndDirection_01, defaultValue=<rgba -1.5 0 0 1>}.RGB) {g=true, r=true}
  n23 = TextureSample(UVs=Add(A=Multiply(A=@n17, B=@n20), B=Multiply(A=@n21, B=@n22))) {texture=&/Cloud_Noise}
  n24 = Multiply(A=Subtract(A=Multiply(A=@n23.RGB) {constB=2}), B=ScalarParameter() {defaultValue=0.05, parameterName=NoiseDistortion_Str})
  n25 = TextureSample(UVs=Add(A=Add(A=@n18, B=@n19), B=ComponentMask(None=@n24) {g=true, r=true})) {texture=&Cloud_Noise}
  NamedRerouteDeclaration(None=Multiply(A=@n25.G, B=ScalarParameter() {defaultValue=0.2, parameterName=IntersectionDistortion_Str})) {name=IntersectionDistortion, nodeColor=<rgba 0 1 0.8359 1>}
  n26 = Saturate(None=OneMinus(None=ScalarParameter() {bUseCustomPrimitiveData=true, parameterName=B_ActivateShield}))
  NamedRerouteDeclaration(None=SphereMask(
    A=NamedRerouteUsage()
    B=Constant()
    Radius=Multiply(B=@n26) {constA=0.5}
    Hardness=ScalarParameter() {defaultValue=1, parameterName=ShieldActivateHardness}
   ) {attenuationRadius=0.5, hardnessPercent=1}) {name=ActivationMask, nodeColor=<rgba 1 0.0703 0 1>}
  n27 = MaterialFunctionCall!(
   Distance Bias=ScalarParameter() {defaultValue=50, group=IntersectingEdge, parameterName=IntersectingEdgeBias}
   Distance Divisor=ScalarParameter() {defaultValue=512, group=IntersectingEdge, parameterName=IntersectingEdgeDivisor}
   Scene Depth
   Pixel Depth
  ) {materialFunction=/Game/Effects/Fort_Effects/Materials/_Functions/MF_DistanceBehindTranslucency_01}
  n28 = Subtract(A=@n27.Normalized Inverted, B=NamedRerouteUsage())
  NamedRerouteDeclaration(None=Power(Base=Saturate(None=@n28), Exp=ScalarParameter() {defaultValue=4, parameterName=IntersectionExponent_02}) {constExponent=12}) {name=IntersectionEdge_Test, nodeColor=<rgba 1 0 0.0703 1>}
  NamedRerouteDeclaration(None=Multiply(A=@n25.G, B=ScalarParameter() {defaultValue=0.6, parameterName=BaseDistortion_Str})) {name=BaseDistortion, nodeColor=<rgba 1 0 0.8906 1>}
  n29 = MaterialFunctionCall(Float2=NamedRerouteUsage()) {materialFunction=&BreakOutFloat2Components}
  n30 = Add(
   A=Multiply(
    A=NamedRerouteDeclaration(None=@n29.R) {name=RadialGradient, nodeColor=<rgba 1 0 0.6094 1>}
    B=ScalarParameter() {defaultValue=5, parameterName=RingCount}
   )
   B=Multiply(A=Time(), B=ScalarParameter() {defaultValue=-3, parameterName=PulseSpeed})
  ) {constB=0}
  n31 = Frac(None=@n30)
  n32 = Multiply(A=OneMinus(None=@n31), B=NamedRerouteUsage())
  NamedRerouteDeclaration(None=SphereMask(
    A=Add(A=@n31, B=@n32)
    B=Constant() {r=1}
    Radius=ScalarParameter() {defaultValue=0.35, parameterName=RingRadius}
    Hardness=ScalarParameter() {parameterName=RingHardness}
   ) {attenuationRadius=0.5, hardnessPercent=1}) {name=PulseRings, nodeColor=<rgba 0 0.3672 1 1>}
  NamedRerouteDeclaration(None=Power(Base=@n27.Normalized Inverted, Exp=ScalarParameter() {defaultValue=12, group=IntersectingEdge, parameterName=Intersection Edge Exponent})) {name=IntersectionEdgeMask, nodeColor=<rgba 0.1719 0 1 1>}
  n33 = Subtract(A=Multiply(A=@n25.RGB) {constB=2})
  n34 = ScalarParameter() {defaultValue=0.05, parameterName=BaseUV_Distortion_Str}
  NamedRerouteDeclaration(None=ComponentMask(None=Multiply(A=@n33, B=@n34)) {g=true, r=true}) {name=BaseUV_Distortion, nodeColor=<rgba 0.5703 1 0 1>}
  NamedRerouteDeclaration(None=SphereMask(
    A=NamedRerouteUsage()
    B=Constant2Vector() {g=0.5, r=0.5}
    Radius=ScalarParameter() {defaultValue=0.5, parameterName=OpacityRadius}
    Hardness=ScalarParameter() {parameterName=OpacityHardness}
   )) {name=RadialOpacity, nodeColor=<rgba 0.8516 1 0 1>}
  NamedRerouteDeclaration(None=Sine(None=@n30)) {name=WPO_PulseRings, nodeColor=<rgba 1 0 0.5391 1>}
  n35 = DotProduct(A=CameraVectorWS(), B=PixelNormalWS())
  n36 = ScalarParameter() {defaultValue=2, parameterName=FresnelStr}
  n37 = Saturate(None=Power(Base=Multiply(A=@n35, B=@n36), Exp=ScalarParameter() {defaultValue=2, parameterName=FresnelPow}))
  NamedRerouteDeclaration(None=OneMinus(None=NamedRerouteDeclaration(None=@n37) {name=FresnelMask, nodeColor=<rgba 1 0.9141 0 1>})) {name=CameraAngleMask, nodeColor=<rgba 0.4766 0 1 1>}
  NamedRerouteUsage()
  NamedRerouteUsage()
  NamedRerouteUsage()
  NamedRerouteUsage()
  NamedRerouteUsage()
  OneMinus(None=NamedRerouteUsage())
  ScalarParameter() {bUseCustomPrimitiveData=true, parameterName=B_ActivateShield}
  Subtract(A=Multiply(A=NamedRerouteUsage()) {constB=2})
 ]
}
