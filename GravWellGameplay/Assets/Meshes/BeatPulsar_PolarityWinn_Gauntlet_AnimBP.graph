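# Animation Blueprint for the BeatPulsar PolarityWinn gauntlet mesh, following the attached parent's pose.
# The copied pose runs through a Control Rig fed four body-type/gender flags, set when character customization loads.
# Additive bone-space transforms on both gauntlet center bones follow via cached poses, then DefaultSlot.
# Each update computes mirrored sine-bob vectors and a wrapping roll, though no link to those transforms is shown.
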AnimBlueprint /GravWellGameplay/Assets/Meshes/BeatPulsar_PolarityWinn_Gauntlet_AnimBP
target_skeleton=&/BeatPulsar_PolarityWinn_Gauntlet_Skeleton
graphs [
 AnimGraph {nodes [
   Output Pose(Result=Slot 'DefaultSlot'(Source=Use cached pose 'FinalPose'()))
   n1 = Control Rig!(
    Source=Copy Pose From Mesh() {bUseAttachedParent=true}
    Alpha
    LargeMale:bool=Get Large Male() -> (Large Male:bool)
    LargeFemale:bool=Get Large Female() -> (Large Female:bool)
    MediumMale:bool=Get Medium Male() -> (Medium Male:bool)
    MediumFemale:bool=Get Medium Female() -> (Medium Female:bool)
   ) -> (Pose) {
    alphaBoolBlend {blendOption=Linear}
    controlRigAssetReference {blueprintRigClass=&/BeatPulsar_PolarityWinn_Gauntlet_CtrlRig.BeatPulsar_PolarityWinn_Gauntlet_CtrlRig_C}
   }
   Save cached pose 'Start Pose'(Pose=@n1) {cacheName=Start Pose}
   n2 = "Transform (Modify) Bone - Bone: gauntlet_center_l"(ComponentPose=Local To Component(LocalPose=Use cached pose 'Start Pose'()), Alpha=1, Translation="0,0,0", Rotation="0,0,0", Scale="1,1,1") {
    rotationMode=BMM_Additive
    rotationSpace=BCS_BoneSpace
    translationMode=BMM_Additive
    translationSpace=BCS_BoneSpace
    alphaBoolBlend {blendOption=Linear}
    boneToModify {boneName=gauntlet_center_l}
   }
   n3 = "Transform (Modify) Bone - Bone: gauntlet_center_r"(ComponentPose=@n2, Alpha=1, Translation="0,0,0", Rotation="0,0,0", Scale="1,1,1") {
    rotationMode=BMM_Additive
    rotationSpace=BCS_BoneSpace
    translationMode=BMM_Additive
    translationSpace=BCS_BoneSpace
    alphaBoolBlend {blendOption=Linear}
    boneToModify {boneName=gauntlet_center_r}
   }
   Save cached pose 'FinalPose'(Pose=Component To Local(ComponentPose=@n3)) {cacheName=FinalPose}
  ]}
 EventGraph {nodes [
   Event BlueprintUpdateAnimation()
   > n1 = Sequence()
   n2 = Set radial(execute=@n1.then_1, radial:double=float - float(A=Get radial() -> (radial:double), B=1)) -> (Output_Get:double)
   > Set rot(rot:Rotator=MakeRotator(Roll=@n2)) -> (Output_Get:Rotator)
   > Branch(Condition="float <= float"(A=Get radial(), B=-360) -> (ReturnValue=~))
   > Set radial(radial=float + float(A=Get radial(), B=360))
   n3 = GetGameTimeInSeconds()
   n4 = float * float(A=0.8, B=-1)
   n5 = float * float(A="Sin (Radians)"(A=float * float(A=@n4, B=@n3)), B=0.6)
   n6 = Set Float Left(execute=@n1.then_0, Float Left:Vector=MakeVector(X=float + float(A=@n5, B=1), Y=0)) -> (Output_Get:Vector)
   > Set Float Right(Float Right:Vector=vector * vector(A=@n6, B:Vector="-1,0,0") -> (ReturnValue=~)) -> (Output_Get:Vector)
   n7 = Event OnCharacterCustomizationLoaded()
   n8 = Get Array Item with EFort Custom Part Type(Array:Array of Custom Character Part Object References=@n7.CustomParts) -> (Output:Custom Character Part Object Reference)
   n9 = "Equal (Enum)"(
    A:EFortCustomBodyType Enum=Get BodyTypesPermitted!(self:Custom Character Part Object Reference=@n8) -> (BodyTypesPermitted:EFortCustomBodyType Enum=NONE)
    B:EFortCustomBodyType Enum=Large
   )
   n10 = "Equal (Enum)"(
    A:EFortCustomGender Enum=Get GenderPermitted!(self:Custom Character Part Object Reference=@n8) -> (GenderPermitted:EFortCustomGender Enum=Invalid)
    B:EFortCustomGender Enum=Male
   )
   > Set Large Male(Large Male:bool=AND Boolean(A=@n9, B=@n10)) -> (Output_Get:bool)
   n11 = "Equal (Enum)"(A=Get GenderPermitted(self=@n8) -> (GenderPermitted=Invalid), B=Female)
   > Set Large Female(Large Female:bool=AND Boolean(A=@n9, B=@n11)) -> (Output_Get:bool)
   n12 = "Equal (Enum)"(A:EFortCustomBodyType Enum=Get BodyTypesPermitted(self=@n8) -> (BodyTypesPermitted=NONE), B:EFortCustomBodyType Enum=Medium)
   > Set Medium Male(Medium Male:bool=AND Boolean(A=@n12, B=@n10)) -> (Output_Get:bool)
   > Set Medium Female(Medium Female:bool=AND Boolean(A=@n12, B=@n11)) -> (Output_Get:bool)
  ]}
]
variables [
 Float Left {type=Vector}
 Float Right {type=Vector}
 Large Female {type=boolean}
 Large Male {type=boolean}
 Medium Female {type=boolean}
 Medium Male {type=boolean}
 radial {type=number}
 rot {type=Rotator}
 Time {type=number}
]
